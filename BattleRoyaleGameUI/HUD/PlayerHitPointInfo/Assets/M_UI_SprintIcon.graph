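# Translucent UI material that draws the sprint icon with speedlines for the Battle Royale player hit point HUD.
# The icon texture's green channel blends adaptive Dim to White color, while boosted red supplies the speedlines mask.
# A soft shadow mask blends four diagonally offset samples of the scaled icon.
# Opacity layers a faint shadow under the icon and speedlines, which fade vertically along an eased UI-UV gradient.
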Material /BattleRoyaleGameUI/HUD/PlayerHitPointInfo/Assets/M_UI_SprintIcon
graph {
 properties {bAutomaticallySetUsageInEditor=false, blendMode=BLEND_Translucent, materialDomain=MD_UI, usedShadingModels=Default Lit}
 nodes [
  n1 = MaterialFunctionCall!(BufferSource) {materialFunction=/BattleRoyaleGameUI/Shared/MF_UI_AdaptiveColors}
  n2 = NamedRerouteUsage()
  n3 = NamedRerouteUsage()
  n4 = MaterialFunctionCall(UVs=TextureCoordinate(), Texture Scale=ScalarParameter() {defaultValue=0.75, parameterName=OverallIconScale}) {materialFunction=&ScaleUVsByCenter}
  n5 = TextureSample(
   UVs=@n4.UVs
   Tex=NamedRerouteDeclaration(None=TextureObject() {samplerType=SAMPLERTYPE_LinearColor, texture=&/T_UI_SprintIcon}) {name=Icon, nodeColor=<rgba 0 0 0 1>}
  ) {samplerType=SAMPLERTYPE_LinearColor, texture=&T_UI_SprintIcon}
  LinearInterpolate(A=LinearInterpolate(A=@n1.Dim, B=@n1.Dim, Alpha=Subtract(A=@n2, B=@n3)), B=@n1.White, Alpha=@n5.G)
  n6 = NamedRerouteUsage()
  n7 = NamedRerouteUsage()
  n8 = NamedRerouteDeclaration(None=Saturate(None=Multiply(A=@n5.R) {constB=2})) {name=SpeedlinesMask, nodeColor=<rgba 0 0 0 1>}
  n9 = ScalarParameter() {defaultValue=0.75, parameterName=Fadeout_Opacity}
  n10 = ComponentMask(None=MaterialFunctionCall() {materialFunction=&GetUserInterfaceUV}.Normalized UV) {g=true}
  n11 = MaterialFunctionCall("Value (0-1)"=OneMinus(None=@n10)) {materialFunction=&MF_UI_EaseCurves}
  n12 = LinearInterpolate(A=@n9, Alpha=@n11.EaseInSine)
  LinearInterpolate(
   A=Multiply(A=Subtract(A=@n6, B=@n7), B=ScalarParameter() {defaultValue=0.125, parameterName=ShadowOpacity})
   Alpha=Multiply(A=Max(A=@n8, B=@n5.G), B=Saturate(None=@n12))
  )
  n13 = ScalarParameter() {defaultValue=0.025, parameterName=Shadow_Y_Offset}
  n14 = Add(
   A=NamedRerouteUsage()
   B=AppendVector(A=ScalarParameter() {defaultValue=0.025, parameterName=Shadow_X_Offset}, B=Multiply(A=@n13) {constB=-1})
  )
  n15 = NamedRerouteUsage()
  n16 = Multiply(A=ScalarParameter() {defaultValue=0.025, parameterName=Shadow_X_Offset}) {constB=-1}
  n17 = ScalarParameter() {defaultValue=0.025, parameterName=Shadow_Y_Offset}
  n18 = TextureSample(UVs=Add(A=NamedRerouteUsage(), B=AppendVector(A=@n16, B=@n17)), Tex=@n15)
  n19 = NamedRerouteUsage()
  n20 = AppendVector(
   A=ScalarParameter() {defaultValue=0.025, parameterName=Shadow_X_Offset}
   B=ScalarParameter() {defaultValue=0.025, parameterName=Shadow_Y_Offset}
  )
  n21 = Multiply(A=@n20) {constB=-1}
  n22 = NamedRerouteDeclaration(None=MaterialFunctionCall(UVs=@n4.UVs, Texture Scale=ScalarParameter() {defaultValue=1, parameterName=ShadowScale}) {materialFunction=&ScaleUVsByCenter}.UVs) {name=UVs, nodeColor=<rgba 0 0 0 1>}
  n23 = LinearInterpolate(A=TextureSample(UVs=Add(A=@n19, B=@n21), Tex=@n15).G, Alpha=TextureSample(UVs=Add(A=@n22, B=@n20), Tex=@n15).G)
  NamedRerouteDeclaration(None=LinearInterpolate(A=TextureSample(UVs=@n14, Tex=@n15).G, Alpha=LinearInterpolate(A=@n18.G, Alpha=@n23))) {name=ShadowMask, nodeColor=<rgba 0 0 0 1>}
  Saturate(None=LinearInterpolate(B=@n9, Alpha=@n10) {constA=2, constB=0})
  SmoothStep(
   Min=ScalarParameter() {defaultValue=0.1, parameterName=Gradient_MinPosition}
   Max=ScalarParameter() {defaultValue=0.95, parameterName=Gradient_MaxPosition}
   Value=@n11.EaseInSine
  )
 ]
}
