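# Translucent UI material for the player HUD stamina bar, drawn as a Progress-driven fill over an adaptive-color background.
# MF_UI_AdaptiveColors supplies dark, dim and bright tones, with the bright tone mixed toward shared white for the fill.
# Anti-aliased distance-field masks build the fill, drop shadow, angled marker and delta/highlight segments for stamina changes.
# Static switch IsGlow outputs a rounded-box glow instead of the bar, while low-end switchers substitute simpler values.
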Material /BattleRoyaleGameUI/HUD/PlayerHitPointInfo/Assets/M_UI_StaminaBar
graph {
 properties {blendMode=BLEND_Translucent, materialDomain=MD_UI, pixelDepthOffsetMode=PDOM_Legacy, usedShadingModels=Default Lit}
 nodes [
  n1 = MaterialFunctionCall!(BufferSource) {materialFunction=/BattleRoyaleGameUI/Shared/MF_UI_AdaptiveColors}
  NamedRerouteDeclaration(None=@n1.Dark) {name=PostProcessDark, nodeColor=<rgba 0 0 0 1>}
  NamedRerouteDeclaration(None=@n1.Dim) {name=PostProcessDim, nodeColor=<rgba 0 0 0 1>}
  n2 = ComponentMask(None=NamedRerouteUsage()) {r=true}
  n3 = ScalarParameter() {defaultValue=0.6, group=Fill, parameterName=Progress, sortPriority=0}
  n4 = Max(A=ScalarParameter() {group=Damage, parameterName=DeltaProgress, sortPriority=0}, B=@n3)
  n5 = Max(A=@n3, B=ScalarParameter() {group=Damage, parameterName=HighlightProgress, sortPriority=0})
  n6 = MaterialFunctionCall!(A=Max(A=@n4, B=@n5), B=Constant()) {materialFunction=/FortUILibrary/MaterialFunctions/Basics/MF_UI_BooleanOperations}
  n7 = LinearInterpolate(
   A=MaterialFunctionCall(DistanceField=Subtract(A=@n2, B=@n4)) {materialFunction=&MF_UI_AntiAliasing}.Inner
   Alpha=@n6.IsLessOrEqual
  ) {constB=0}
  n8 = LinearInterpolate(
   A=MaterialFunctionCall(DistanceField=Subtract(A=@n2, B=@n5)) {materialFunction=&MF_UI_AntiAliasing}.Inner
   Alpha=@n6.IsLessOrEqual
  ) {constB=0}
  n9 = Multiply(
   A=NamedRerouteDeclaration(None=Max(A=@n7, B=@n8)) {name=DeltaMask, nodeColor=<rgba 0 0 0 1>}
   B=VectorParameter() {group=Color, parameterName=DeltaColor, sortPriority=4, defaultValue=<rgba 0.9301 0.0865 0.0185 1>}.RGB
  )
  n10 = VectorParameter() {group=Color, parameterName=HighlightColor, sortPriority=4, defaultValue=<rgba 0.1878 0.6445 0.0144 1>}
  n11 = NamedRerouteUsage()
  n12 = ScalarParameter() {defaultValue=0.1, group=Damage, parameterName=DeltaShadowOpacity, sortPriority=1}
  NamedRerouteDeclaration(None=Subtract(A=LinearInterpolate(A=@n9, B=@n10.RGB, Alpha=@n8), B=Multiply(A=@n11, B=@n12))) {name=DeltaColor, nodeColor=<rgba 0 0 0 1>}
  n13 = ScalarParameter() {defaultValue=0.6, group=Fill, parameterName=Progress, sortPriority=0}
  n14 = ScalarParameter() {parameterName=BarPadding}
  n15 = MaterialFunctionCall() {materialFunction=&GetUserInterfaceUV}
  n16 = Multiply(A=@n14, B=Divide(A=ComponentMask(None=@n15.Pixel Size) {r=true}, B=ComponentMask(None=@n15.Pixel Size) {g=true}))
  n17 = MaterialFunctionCall(
   UVs=MaterialFunctionCall() {materialFunction=&GetUserInterfaceUV}.Normalized UV
   Texture Scale=OneMinus(None=AppendVector(A=@n14, B=@n16))
  ) {materialFunction=&ScaleUVsByCenter}
  n18 = Subtract(A=ComponentMask(None=NamedRerouteDeclaration(None=@n17.UVs) {name=BarUVs, nodeColor=<rgba 0 0 0 1>}) {r=true}, B=@n13) {constB=0.75}
  n19 = SmoothStep(Min=Add(A=@n13, B=ScalarParameter() {defaultValue=0.25, parameterName=ProgressFill_DropShadowSize}), Value=OneMinus(None=@n18))
  NamedRerouteDeclaration(None=MaterialFunctionCall("Value (0-1)"=Saturate(None=@n19)) {materialFunction=&MF_UI_EaseCurves}.EaseInQuint) {name=ProgressFillDropShadowMask, nodeColor=<rgba 0 0 0 1>}
  n20 = ScalarParameter() {defaultValue=0.6, group=Fill, parameterName=Progress, sortPriority=0}
  n21 = Multiply(A=OneMinus(None=@n20)) {constB=0.5}
  n22 = Constant()
  n23 = MaterialFunctionCall(
   UVs=Add(A=MaterialFunctionCall() {materialFunction=&GetUserInterfaceUV}.Normalized UV, B=AppendVector(A=@n21, B=@n22))
   Size=AppendVector(A=@n20, B=Constant() {r=1})
   CornerRoundness=ScalarParameter() {defaultValue=0.5, parameterName=GlowCornerRoundess}
   KeepAspectRatio=StaticBool() {value=true}
  ) {materialFunction=&MF_UI_Box}
  n24 = ScalarParameter() {defaultValue=-0.25, parameterName=GlowThreshold}
  n25 = ScalarParameter() {defaultValue=0.5, parameterName=GlowSmoothing}
  NamedRerouteDeclaration(None=Saturate(None=MaterialFunctionCall(DistanceField=@n23.SDF, Threshold=@n24, Smoothing=@n25) {materialFunction=&MF_UI_AntiAliasing}.Inner)) {name=GlowMask, nodeColor=<rgba 0 0 0 1>}
  n26 = NamedRerouteDeclaration(None=@n18) {name=ProgressFillSDF, nodeColor=<rgba 0 0 0 1>}
  NamedRerouteDeclaration(None=MaterialFunctionCall(DistanceField=@n26) {materialFunction=&MF_UI_AntiAliasing}.Inner) {name=ProgressFillMask, nodeColor=<rgba 0 0 0 1>}
  NamedRerouteDeclaration(None=MaterialFunctionCall(DistanceField=@n26, Smoothing=ScalarParameter() {defaultValue=0.125, parameterName=MarkerShadowAmount}) {materialFunction=&MF_UI_AntiAliasing}.Inner) {name=MarkerShadowMask, nodeColor=<rgba 0 0 0 1>}
  n27 = ComponentMask(None=MaterialFunctionCall() {materialFunction=&GetUserInterfaceUV}.Normalized UV) {g=true}
  n28 = Add(
   A=NamedRerouteUsage()
   B=Multiply(A=Subtract(A=@n27) {constB=0.5}, B=ScalarParameter() {defaultValue=0.0125, group=Marker, parameterName=MarkerAngle, sortPriority=0})
  )
  NamedRerouteDeclaration(None=MaterialFunctionCall(
    DistanceField=Abs(None=@n28)
    Threshold=ScalarParameter() {defaultValue=0.02, group=Marker, parameterName=MarkerWidth, sortPriority=1}
    Smoothing=ScalarParameter() {group=Marker, parameterName=MarkerSmoothing, sortPriority=2}
   ) {materialFunction=&MF_UI_AntiAliasing}.Inner) {name=MarkerMask, nodeColor=<rgba 0 0 0 1>}
  n29 = ComponentMask(None=CollectionParameter() {collection=/BattleRoyaleGameUI/Shared/MPC_BRGameUI, parameterName=White}) {b=true, g=true, r=true}
  n30 = Multiply(
   A=LinearInterpolate(
    A=NamedRerouteDeclaration(None=@n1.Bright) {name=PostProcessBright, nodeColor=<rgba 0 0 0 1>}
    B=@n29
    Alpha=ScalarParameter() {defaultValue=0.25, parameterName=Fill_ColorWhiteMix}
   )
   B=NamedRerouteUsage()
  )
  n31 = Multiply(A=@n29, B=NamedRerouteUsage())
  NamedRerouteDeclaration(None=Saturate(None=Add(A=@n30, B=@n31))) {name=FillColor, nodeColor=<rgba 0 0 0 1>}
  n32 = LinearInterpolate(
   A=MaterialFunctionCall!(Default=Constant() {r=0.85}, LowEnd=Constant() {r=0.525}) {materialFunction=/BattleRoyaleGameUI/Shared/MF_UI_LowEnd_ValueSwitcher}.Result
   B=MaterialFunctionCall(Default=ScalarParameter() {defaultValue=0.75, group=Color, parameterName=BackgroundOpacity, sortPriority=7}, LowEnd=Constant() {r=0.464}) {materialFunction=&MF_UI_LowEnd_ValueSwitcher}.Result
   Alpha=OneMinus(None=NamedRerouteUsage())
  ) {constA=0.85}
  n33 = MaterialFunctionCall(Default=Constant() {r=0.5}, LowEnd=Constant() {r=0.268}) {materialFunction=&MF_UI_LowEnd_ValueSwitcher}
  n34 = NamedRerouteDeclaration(None=MaterialFunctionCall(A=ScalarParameter() {defaultValue=0.6, group=Fill, parameterName=Progress, sortPriority=0}, B=Constant()) {materialFunction=&MF_UI_BooleanOperations}.IsLessOrEqual) {name=ProgressEmptyLogic, nodeColor=<rgba 0 0 0 1>}
  StaticSwitchParameter(
   True=Multiply(A=NamedRerouteUsage(), B=ScalarParameter() {defaultValue=0.5, parameterName=GlowOpacity})
   False=LinearInterpolate(
    A=LinearInterpolate(A=@n32, B=@n33.Result, Alpha=@n34) {constB=0.6667}
    B=ScalarParameter() {defaultValue=1, parameterName=FillTranslucency}
    Alpha=NamedRerouteUsage()
   )
  ) {parameterName=IsGlow}
  n35 = NamedRerouteUsage()
  n36 = MaterialFunctionCall(Default=NamedRerouteUsage(), LowEnd=NamedRerouteUsage()) {materialFunction=&MF_UI_LowEnd_ValueSwitcher}
  n37 = NamedRerouteUsage()
  n38 = LinearInterpolate(A=NamedRerouteUsage(), B=NamedRerouteUsage(), Alpha=NamedRerouteUsage())
  n39 = MaterialFunctionCall(Default=ScalarParameter() {defaultValue=1, parameterName=FillOpacity}, LowEnd=Constant() {r=0.9}) {materialFunction=&MF_UI_LowEnd_ValueSwitcher}
  StaticSwitchParameter(
   True=MaterialFunctionCall(Default=NamedRerouteUsage(), LowEnd=NamedRerouteUsage()) {materialFunction=&MF_UI_LowEnd_ValueSwitcher}.Result
   False=LinearInterpolate(A=LinearInterpolate(A=@n35, B=@n36.Result, Alpha=@n37), B=Multiply(A=@n38, B=@n39.Result), Alpha=NamedRerouteUsage())
  ) {parameterName=IsGlow}
 ]
}
