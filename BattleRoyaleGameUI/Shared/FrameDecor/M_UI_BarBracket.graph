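# Translucent UI material that draws a decorative bar bracket reacting to a Progress value.
# The shape is a rotated, perspective-warped triangle SDF minus a trapezoid whose UV is clamped at mid-width, with antialiased edges and a soft glow mask.
# Color blends adaptive dim and bright tones, shifting through shared red and orange as Progress falls below LowThreshold.
# Opacity uses an eased vertical gradient, low-progress and empty-state boosts, animation parameters and low-end device switches.
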Material /BattleRoyaleGameUI/Shared/FrameDecor/M_UI_BarBracket
graph {
 properties {bAutomaticallySetUsageInEditor=false, blendMode=BLEND_Translucent, materialDomain=MD_UI, usedShadingModels=Default Lit}
 nodes [
  n1 = MaterialFunctionCall!(
   UV=Subtract(
    A=MaterialFunctionCall() {materialFunction=&GetUserInterfaceUV}.Normalized UV
    B=Constant2Vector() {g=-0.075, r=-0.15}
   )
   Perspective Rotation=Constant() {r=0.88}
   Perspective Amount=Constant() {r=0.48}
  ) {materialFunction=/FortUILibrary/MaterialFunctions/Basics/MF_UI_Perspective}
  n2 = Constant() {r=0.125}
  n3 = MaterialFunctionCall() {materialFunction=&GetUserInterfaceUV}
  n4 = Min(A=NamedRerouteDeclaration(None=ComponentMask(None=@n3.Normalized UV) {r=true}) {name=MaskR, nodeColor=<rgba 0 0 0 1>}) {constB=0.5}
  n5 = ComponentMask(None=@n3.Normalized UV) {g=true}
  n6 = MaterialFunctionCall!(
   DistanceFieldA=MaterialFunctionCall!(UVs=Rotator(Coordinate=@n1.Perspective UVs, Time=@n2) {speed=3.1416}, Radius=Constant() {r=1.7}, Sides=Constant() {r=3}, InnerAngle) {materialFunction=/FortUILibrary/MaterialFunctions/Shapes/SDFs/MF_SDF_RegularPolygon}.SDF
   DistanceFieldB=MaterialFunctionCall!(UVs=AppendVector(A=@n4, B=@n5), TopWidth=Constant() {r=0.35}, BottomWidth=Constant() {r=0.475}, Height=Constant() {r=0.35}) {materialFunction=/FortUILibrary/MaterialFunctions/Shapes/SDFs/MF_SDF_IsoscelesTrapezoid}.SDF
   BlendAmount
  ) {materialFunction=/FortUILibrary/MaterialFunctions/Shapes/Modifiers/MF_UI_SDF_Subtraction}
  n7 = MaterialFunctionCall(DistanceField=@n6.Exact) {materialFunction=&MF_UI_AntiAliasing}
  n8 = ScalarParameter() {defaultValue=1, parameterName=Opacity}
  n9 = ScalarParameter() {defaultValue=0.25, parameterName=OpacityTop}
  n10 = MaterialFunctionCall("Value (0-1)"=@n5) {materialFunction=&MF_UI_EaseCurves}
  n11 = LinearInterpolate(
   A=MaterialFunctionCall!(Default=@n8, LowEnd=Multiply(A=@n8) {constB=0.5}) {materialFunction=/BattleRoyaleGameUI/Shared/MF_UI_LowEnd_ValueSwitcher}.Result
   B=MaterialFunctionCall(Default=@n9, LowEnd=Multiply(A=@n9) {constB=0.5}) {materialFunction=&MF_UI_LowEnd_ValueSwitcher}.Result
   Alpha=SmoothStep(
    Min=ScalarParameter() {parameterName=Alpha_MaxPosition}
    Max=ScalarParameter() {defaultValue=1, parameterName=Alpha_MinPosition}
    Value=OneMinus(None=@n10.EaseInCubic)
   )
  ) {constA=1, constB=0}
  n12 = ScalarParameter() {defaultValue=1.5, parameterName=LowOpacity}
  n13 = Constant() {r=3}
  n14 = Saturate(None=Multiply(A=@n11, B=MaterialFunctionCall(Default=@n12, LowEnd=@n13) {materialFunction=&MF_UI_LowEnd_ValueSwitcher}.Result))
  n15 = NamedRerouteUsage()
  n16 = LinearInterpolate(A=@n11, B=Saturate(None=LinearInterpolate(A=@n14, B=@n11, Alpha=@n15)), Alpha=NamedRerouteUsage())
  n17 = NamedRerouteDeclaration(None=MaterialFunctionCall!(A=ScalarParameter() {defaultValue=0.125, parameterName=Progress}, B=Constant()) {materialFunction=/FortUILibrary/MaterialFunctions/Basics/MF_UI_BooleanOperations}.IsLessOrEqual) {name=EmptyStateLogic, nodeColor=<rgba 0 0 0 1>}
  n18 = Multiply(A=@n7.Inner, B=LinearInterpolate(A=LinearInterpolate(A=@n16, B=@n14, Alpha=@n17), Alpha=ScalarParameter() {parameterName=Anim_Opacity}))
  n19 = NamedRerouteUsage()
  n20 = ScalarParameter() {defaultValue=0.5, parameterName=GlowOpacity}
  LinearInterpolate(A=@n18, B=LinearInterpolate(A=Multiply(A=@n19, B=@n20), Alpha=@n18), Alpha=NamedRerouteUsage())
  n21 = MaterialFunctionCall!(BufferSource) {materialFunction=/BattleRoyaleGameUI/Shared/MF_UI_AdaptiveColors}
  n22 = LinearInterpolate(
   A=MaterialFunctionCall(Default=@n21.Dim, LowEnd=@n21.Dim) {materialFunction=&MF_UI_LowEnd_ValueSwitcher}.Result
   B=@n21.Bright
   Alpha=ScalarParameter() {parameterName=Anim_Color}
  )
  n23 = ComponentMask(None=CollectionParameter() {collection=/BattleRoyaleGameUI/Shared/MPC_BRGameUI, parameterName=Red}) {b=true, g=true, r=true}
  n24 = ComponentMask(None=CollectionParameter() {collection=&MPC_BRGameUI, parameterName=Orange}) {b=true, g=true, r=true}
  n25 = LinearInterpolate(A=@n23, B=@n24)
  n26 = ScalarParameter() {defaultValue=0.125, parameterName=Progress}
  n27 = Max(A=ScalarParameter() {defaultValue=0.25, parameterName=LowThreshold}) {constB=0.001}
  n28 = NamedRerouteDeclaration(None=MaterialFunctionCall("Value (0-1)"=Divide(A=@n26, B=@n27)) {materialFunction=&MF_UI_EaseCurves}.EaseInSine) {name=LowThresholdCurve, nodeColor=<rgba 0 0 0 1>}
  n29 = MaterialFunctionCall(A=ScalarParameter() {defaultValue=0.125, parameterName=Progress}, B=ScalarParameter() {defaultValue=0.25, parameterName=LowThreshold}) {materialFunction=&MF_UI_BooleanOperations}
  LinearInterpolate(
   A=LinearInterpolate(
    A=@n22
    B=MaterialFunctionCall(1=@n25, 2=@n24, 3=@n22, A=@n28) {materialFunction=&Lerp_Multiple_Float3}.Lerp 3 Inputs
    Alpha=NamedRerouteDeclaration(None=@n29.IsLessOrEqual) {name=LowThresholdLogic, nodeColor=<rgba 0 0 0 1>}
   )
   B=@n23
   Alpha=NamedRerouteUsage()
  )
  n30 = NamedRerouteUsage()
  n31 = MaterialFunctionCall(
   DistanceField=@n6.Smooth
   Threshold=ScalarParameter() {parameterName=Threshold}
   Smoothing=ScalarParameter() {defaultValue=0.5, parameterName=GlowSmoothing}
  ) {materialFunction=&MF_UI_AntiAliasing}
  NamedRerouteDeclaration(None=Multiply(A=OneMinus(None=@n30), B=Saturate(None=@n31.Inner))) {name=GlowMask, nodeColor=<rgba 0 0 0 1>}
  Saturate(None=@n7.Inner)
 ]
}
