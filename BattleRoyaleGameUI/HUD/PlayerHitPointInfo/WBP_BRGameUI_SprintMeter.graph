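# HUD widget that shows the tactical sprint stamina meter, appearing when energy drops below full and fading out once refilled.
# Energy component delegates drive material parameters on bar, glow and brackets, with timer-based delta and tagged highlighted-gain effects and a low-energy pulse.
# Gameplay tag waits force it visible or track riding; vehicle, riding, creative quickbar and DBNO states hide it.
# It registers as a tagged widget host, letting external content replace the default bar.
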WidgetBlueprint /BattleRoyaleGameUI/HUD/PlayerHitPointInfo/WBP_BRGameUI_SprintMeter
animations [
 AnimFullBar {
  binding_count=1
  length_seconds=0.75
  object=&self:AnimFullBar
  sequence {bindings [Overlay_Energy {
     id=#1
     possessed_class=Overlay
     tracks [{
       type=MovieSceneFloatTrack
       display_name=Render Opacity
       sections [{type=MovieSceneFloatSection, end_frame=15, start_frame=0, channels [None {type=MovieSceneScriptingFloatChannel, keys [[0,1],[10,1],[15,0]]}]}]
      }]
    }]}
 }
 AnimLow {
  binding_count=1
  length_seconds=0.5
  object=&self:AnimLow
  sequence {bindings [Overlay_Energy {
     id=#2
     possessed_class=Overlay
     tracks [{
       type=MovieSceneFloatTrack
       display_name=Render Opacity
       sections [{
         type=MovieSceneFloatSection
         end_frame=10
         start_frame=0
         channels [None {type=MovieSceneScriptingFloatChannel, keys [[0,1],[5,0.33],[10,1]]}]
        }]
      }]
    }]}
 }
 AnimIntro {
  binding_count=3
  length_seconds=0.5
  object=&self:AnimIntro
  sequence {bindings [
    Image_StaminaBar {
     id=#3
     possessed_class=Image
     tracks [
      {
       type=MovieSceneFloatTrack
       display_name=Render Opacity
       sections [{type=MovieSceneFloatSection, channels [None {type=MovieSceneScriptingFloatChannel, keys [[0,0],[10,1]]}]}]
      }
      {
       type=MovieScene2DTransformTrack
       sections [{
         type=MovieScene2DTransformSection
         channels [
          Translation.X {type=MovieSceneScriptingFloatChannel, keys [[0,-128],[8,0]]}
          Translation.Y {type=MovieSceneScriptingFloatChannel, const [0,1]}
          Scale.X {type=MovieSceneScriptingFloatChannel, const [1,1]}
          Scale.Y {type=MovieSceneScriptingFloatChannel, const [1,1]}
         ]
        }]
      }
     ]
    }
    Image_SprintIcon {
     id=#4
     possessed_class=Image
     tracks [
      {
       type=MovieSceneFloatTrack
       display_name=Render Opacity
       sections [{type=MovieSceneFloatSection, channels [None {type=MovieSceneScriptingFloatChannel, keys [[2,0],[10,1]]}]}]
      }
      {
       type=MovieScene2DTransformTrack
       sections [{
         type=MovieScene2DTransformSection
         channels [
          Translation.X {type=MovieSceneScriptingFloatChannel, keys [[2,-64],[10,-20]]}
          Translation.Y {type=MovieSceneScriptingFloatChannel, keys [[2,12],[10,6]]}
          Scale.X {type=MovieSceneScriptingFloatChannel, keys [[2,1],[10,1.5]]}
          Scale.Y {type=MovieSceneScriptingFloatChannel, keys [[2,1],[10,1.5]]}
         ]
        }]
      }
     ]
    }
    Image_Bracket {
     id=#5
     possessed_class=Image
     tracks [{
       type=MovieSceneFloatTrack
       display_name=Render Opacity
       sections [{type=MovieSceneFloatSection, channels [None {type=MovieSceneScriptingFloatChannel, keys [[5,0],[10,1]]}]}]
      }]
    }
   ]}
 }
 Anim_Intro_CenterScreenVersion {
  binding_count=3
  length_seconds=0.6
  object=&self:Anim_Intro_CenterScreenVersion
  sequence {bindings [
    Overlay_Energy {
     id=#6
     possessed_class=Overlay
     tracks [
      {
       type=MovieSceneFloatTrack
       display_name=Render Opacity
       sections [{type=MovieSceneFloatSection, channels [None {type=MovieSceneScriptingFloatChannel, keys [[0,0],[2,1]]}]}]
      }
      {
       type=MovieScene2DTransformTrack
       sections [{
         type=MovieScene2DTransformSection
         channels [
          Translation.Y {type=MovieSceneScriptingFloatChannel, keys [[0,80],[6,0]]}
          Angle {type=MovieSceneScriptingFloatChannel, const [0,1]}
          Scale.X {type=MovieSceneScriptingFloatChannel, keys [[0,4],[4,1]]}
          Scale.Y {type=MovieSceneScriptingFloatChannel, keys [[0,2],[4,1]]}
         ]
        }]
      }
     ]
    }
    Image_Bracket {
     id=#7
     possessed_class=Image
     tracks [
      {
       type=MovieScene2DTransformTrack
       sections [{type=MovieScene2DTransformSection, channels [Translation.X {type=MovieSceneScriptingFloatChannel, keys [[2,-64],[8,0]]}]}]
      }
      {
       type=MovieSceneWidgetMaterialTrack
       sections [{
         type=MovieSceneParameterSection
         channels [
          Anim_Color {type=MovieSceneScriptingFloatChannel, keys [[6,1],[12,0]]}
          Anim_Opacity {type=MovieSceneScriptingFloatChannel, keys [[2,1],[12,0]]}
         ]
        }]
      }
     ]
    }
    Image_Bracket_Rght {
     id=#8
     possessed_class=Image
     tracks [
      {
       type=MovieSceneWidgetMaterialTrack
       sections [{
         type=MovieSceneParameterSection
         channels [
          Anim_Color {type=MovieSceneScriptingFloatChannel, keys [[6,1],[12,0]]}
          Anim_Opacity {type=MovieSceneScriptingFloatChannel, keys [[2,1],[12,0]]}
         ]
        }]
      }
      {
       type=MovieScene2DTransformTrack
       sections [{type=MovieScene2DTransformSection, channels [Translation.X {type=MovieSceneScriptingFloatChannel, keys [[2,64],[8,0]]}]}]
      }
     ]
    }
   ]}
 }
]
graphs [
 AreDBNOEnhancementsEnabled {nodes [
   n1 = AreDBNOEnhancementsEnabled!() -> (then, Pawn:Fort Pawn Object Reference) {is=title K2Node_FunctionEntry}
   n2 = GetComponentByClass!(self:Actor Object Reference=@n1, ComponentClass:Actor Component Class Reference=/Script/FortniteGame.DBNOBehaviorOverrideComponent) -> (ReturnValue:DBNOBehavior Override Component Object Reference)
   > n3 = Is Valid(InputObject=@n2)
   Return Node!(execute=@n3.Is Not Valid, Enabled:bool=false)
   > Return Node(Enabled=AreDBNOEnhancementsEnabled!(self:DBNOBehavior Override Component Object Reference=@n2) -> (ReturnValue:bool) {is=title K2Node_CallFunction})
  ]}
 BindToPlayerStateZoneEvents {nodes [
   BindToPlayerStateZoneEvents!() -> (then, DoBind:bool=~) {is=title K2Node_FunctionEntry}
   > n1 = Set PlayerStateZone(PlayerStateZone:Fort Player State Zone Object Reference=GetPlayerStateZone!(self:Fort HUDContext Object Reference=Get FortHUDContext!() -> (ReturnValue:Fort HUDContext Object Reference) {is=title K2Node_GetBlueprintContextEx}) -> (ReturnValue:Fort Player State Zone Object Reference)) -> (Output_Get:Fort Player State Zone Object Reference)
   > Is Valid(InputObject=@n1)
   > n2 = Sequence()
   > n3 = Branch(Condition=Get DoBind() -> (DoBind:bool))
   n4 = Create Event()
   n5 = Get PlayerStateZone() -> (PlayerStateZone:Fort Player State Zone Object Reference)
   > Bind Event to On Overshield Visibility Changed(self:Fort Player State Zone Object Reference=@n5, Delegate=@n4) {is=title K2Node_AddDelegate}
   > CheckOvershieldVisiblity()
   Unbind Event from On Overshield Visibility Changed(execute=@n3.else, self:Fort Player State Zone Object Reference=@n5, Delegate=@n4) {is=title K2Node_RemoveDelegate}
   n6 = GetPawnFromPlayerStateZone!(execute=@n2.then_1, InActor:Actor Object Reference=@n1) -> (then, OutPawn:Fort Player Pawn Object Reference, PawnReturned:bool)
   > Branch(Condition=@n6.PawnReturned)
   > n7 = Branch(Condition=AreDBNOEnhancementsEnabled!(self:Self Object Reference, Pawn:Fort Pawn Object Reference=@n6) -> (Enabled:bool))
   n8 = Branch(execute=@n7.else, Condition=Get DoBind())
   n9 = Create Event()
   > Bind Event to On DBNOState Changed(self:Fort Pawn Object Reference=@n6, Delegate=@n9) {is=title K2Node_AddDelegate}
   Unbind Event from On DBNOState Changed(execute=@n8.else, self:Fort Pawn Object Reference=@n6, Delegate=@n9) {is=title K2Node_RemoveDelegate}
  ]}
 BindToTacticalSprintComponent {nodes [
   n1 = BindToTacticalSprintComponent!() -> (then, DoBind:bool=~) {is=title K2Node_FunctionEntry}
   > n2 = Branch(Condition=@n1)
   n3 = GetOwningPlayerPawn!(self:User Widget Object Reference) -> (ReturnValue:Pawn Object Reference)
   > Is Valid(InputObject=@n3)
   > n4 = Set BoundEnergyComponents(BoundEnergyComponents:Array of Fort Component Energy Object References=Get Components By Class!(self:Actor Object Reference=@n3, ComponentClass:Actor Component Class Reference=/Script/EnergyRuntime.FortComponent_Energy) -> (ReturnValue:Array of Fort Component Energy Object References)) -> (Output_Get:Array of Fort Component Energy Object References)
   > n5 = For Each Loop(Array:Array of Fort Component Energy Object References=@n4) -> (Array Element:Fort Component Energy Object Reference)
   > Bind Event to On Current Energy Changed(self:Fort Component Energy Object Reference=@n5, Delegate=Create Event()) {is=title K2Node_AddDelegate}
   > Bind Event to On Energy Chunk Added(self:Fort Component Energy Object Reference=@n5, Delegate=Create Event()) {is=title K2Node_AddDelegate}
   > OnCurrentEnergyChanged!(execute, self, EnergyComponent:Fort Component Energy Object Reference=@n5, PreviousValue:float=0) -> (then)
   n6 = For Each Loop(Exec=@n2.else, Array=Get BoundEnergyComponents() -> (BoundEnergyComponents:Array of Fort Component Energy Object References))
   Clear(execute=@n6.Completed, TargetArray:Array of Fort Component Energy Object References=Get BoundEnergyComponents())
   Is Valid(exec=@n6.LoopBody, InputObject=@n6)
   > Unbind Event from On Current Energy Changed(self:Fort Component Energy Object Reference=@n6, Delegate=Create Event()) {is=title K2Node_RemoveDelegate}
   > Unbind Event from On Energy Chunk Added(self:Fort Component Energy Object Reference=@n6, Delegate=Create Event()) {is=title K2Node_RemoveDelegate}
  ]}
 BindToVehicleEvents {nodes [
   n1 = BindToVehicleEvents!() -> (then, DoBind:bool=~) {is=title K2Node_FunctionEntry}
   > n2 = Branch(Condition=@n1)
   n3 = Create Event()
   n4 = Get FortGlobalUIContext() -> (ReturnValue:Fort Global UIContext Object Reference) {is=title K2Node_GetSubsystem}
   > Bind Event to On Exit Vehicle(self:Fort Global UIContext Object Reference=@n4, Delegate=@n3) {is=title K2Node_AddDelegate}
   n5 = Create Event()
   > Bind Event to On Enter Vehicle Driver(self:Fort Global UIContext Object Reference=@n4, Delegate=@n5) {is=title K2Node_AddDelegate}
   n6 = Create Event()
   > Bind Event to On Enter Vehicle Passenger(self:Fort Global UIContext Object Reference=@n4, Delegate=@n6) {is=title K2Node_AddDelegate}
   Unbind Event from On Exit Vehicle(execute=@n2.else, self:Fort Global UIContext Object Reference=@n4, Delegate=@n3) {is=title K2Node_RemoveDelegate}
   > Unbind Event from On Enter Vehicle Driver(self:Fort Global UIContext Object Reference=@n4, Delegate=@n5) {is=title K2Node_RemoveDelegate}
   > Unbind Event from On Enter Vehicle Passenger(self:Fort Global UIContext Object Reference=@n4, Delegate=@n6) {is=title K2Node_RemoveDelegate}
  ]}
 CheckOvershieldVisiblity {nodes [
   CheckOvershieldVisiblity() {is=title K2Node_FunctionEntry}
   > n1 = Branch(Condition=IsUsingTouch(PlayerController=GetOwningPlayer()))
   n2 = GetPlayerStateZone(self=Get FortHUDContext())
   Is Valid(exec=@n1.else, InputObject=@n2)
   > n3 = Branch(Condition=ShouldOvershieldBarBeVisible!(self:Fort Player State Zone Object Reference=@n2) -> (ReturnValue:bool))
   > SetRenderTranslation(self=Get Overlay_Energy() -> (Overlay_Energy:Overlay Object Reference), Translation=(X=0,Y=-5))
   SetRenderTranslation(execute=@n3.else, self=Get Overlay_Energy(), Translation=(X=0,Y=12))
  ]}
 DisableHighlightBar {nodes [
   DisableHighlightBar() {is=title K2Node_FunctionEntry}
   > Set IsGainHighlightActive(IsGainHighlightActive:bool=false) -> (Output_Get:bool)
   > Clear and Invalidate Timer by Handle(Handle=Get EmptyHighlight_Handle() -> (EmptyHighlight_Handle:Timer Handle Structure))
   > Clear and Invalidate Timer by Handle(Handle=Get StartHighlightFadeOut_Handle() -> (StartHighlightFadeOut_Handle:Timer Handle Structure))
  ]}
 EmptyDelta {nodes [
   EmptyDelta() {is=title K2Node_FunctionEntry}
   > n1 = GetDynamicMaterial(self=Get Image_StaminaBar() -> (Image_StaminaBar:Image Object Reference))
   > SetScalarParameterValue(self=@n1, ParameterName=DeltaProgress, Value=0)
  ]}
 EmptyHighlightBar {nodes [
   EmptyHighlightBar() {is=title K2Node_FunctionEntry}
   > n1 = Branch(Condition="Not Equal (Float)"(A=Get ValueBeforeHighlightedGain() -> (ValueBeforeHighlightedGain:double), B=Get PrevEnergy() -> (PrevEnergy:double)))
   n2 = Get EmptyHighlight_StartTime() -> (EmptyHighlight_StartTime:double)
   n3 = Get EmptyHighlight_TotalSeconds() -> (EmptyHighlight_TotalSeconds:double)
   > Set ValueBeforeHighlightedGain(ValueBeforeHighlightedGain:double=MapRangeClamped(
     Value=GetTimeSeconds!() -> (ReturnValue:double)
     InRangeA=Get EmptyHighlight_StartTime()
     InRangeB=float + float(A=@n2, B=@n3)
     OutRangeA=Get ValueBeforeHighlightedGain()
     OutRangeB=Get PrevEnergy()
    )) -> (Output_Get:double)
   > SetStaminaProgress!(execute, self, CurrentEnergy:double=Get PrevEnergy()) -> (then)
   DisableHighlightBar(execute=@n1.else) {is=title K2Node_CallFunction}
  ]}
 EventGraph {nodes [
   n1 = OnEnergyAdded!() -> (
    OutputDelegate
    then
    EnergyComponent:Fort Component Energy Object Reference
    PreviousValue:float=~
    AdditionalTags:Gameplay Tag Container Structure
   ) {is=title K2Node_CustomEvent}
   n2 = Get HighlightGainTag() -> (HighlightGainTag:Gameplay Tag Structure)
   > Branch(Condition=AND Boolean(
     A=IsGameplayTagValid!(GameplayTag:Gameplay Tag Structure=@n2) -> (ReturnValue:bool)
     B=HasTag!(TagContainer:"Gameplay Tag Container Structure (by ref)"=@n1.AdditionalTags, Tag:Gameplay Tag Structure=@n2, bExactMatch:bool=true) -> (ReturnValue:bool)
    ))
   > OnHighlightActivated!(execute, self, EnergyComponent:Fort Component Energy Object Reference=@n1.EnergyComponent, PreviousValue:float=@n1.PreviousValue) -> (then)
   RemoveBinds() {is=title K2Node_CustomEvent}
   > BindToTacticalSprintComponent!(execute, self, DoBind:bool=false) -> (then) {is=title K2Node_CallFunction}
   > BindToPlayerStateZoneEvents!(execute, self, DoBind:bool=false) -> (then) {is=title K2Node_CallFunction}
   > BindToVehicleEvents!(execute, self, DoBind:bool=false) -> (then) {is=title K2Node_CallFunction}
   > Unbind Event from On Creative Quickbar Enabled Changed(
    self:Creative HUDLContext Object Reference=Get CreativeHUDLContext!() -> (ReturnValue:Creative HUDLContext Object Reference) {is=title K2Node_GetBlueprintContextEx}
    Delegate=Create Event()
   ) {is=title K2Node_RemoveDelegate}
   OnPawnSet() {is=title K2Node_CustomEvent}
   > RemoveBinds() {is=title K2Node_CallFunction}
   > SetupEventBinds()
   SetupEventBinds() {is=title K2Node_CustomEvent}
   > BindToTacticalSprintComponent(DoBind=true)
   > BindToPlayerStateZoneEvents(DoBind=true)
   > BindToVehicleEvents(DoBind=true)
   > Bind Event to On Creative Quickbar Equipped Changed(self:Creative HUDLContext Object Reference=Get CreativeHUDLContext(), Delegate=Create Event()) {is=title K2Node_AddDelegate}
   > UpdateVisibilityFromState()
   n3 = GetOwningPlayerPawn()
   > Is Valid(InputObject=@n3)
   > n4 = WaitGameplayTagAddToActor!(
    execute:Exec
    TargetActor:Actor Object Reference=@n3
    Tag:Gameplay Tag Structure=(TagName="Granted.Riding.Player.IsRiding")
    OnlyTriggerOnce:bool=false
   ) -> (then:Exec, AsyncTaskProxy:Ability Async Wait Gameplay Tag Added Object Reference, Added:Exec) {is=title K2Node_AsyncAction}
   > n5 = WaitGameplayTagRemoveFromActor!(
    execute:Exec
    TargetActor:Actor Object Reference=@n3
    Tag:Gameplay Tag Structure=(TagName="Granted.Riding.Player.IsRiding")
    OnlyTriggerOnce:bool=false
   ) -> (then:Exec, AsyncTaskProxy:Ability Async Wait Gameplay Tag Removed Object Reference, Removed:Exec) {is=title K2Node_AsyncAction}
   > n6 = WaitGameplayTagQueryOnActor!(
    execute:Exec
    TargetActor:Actor Object Reference=@n3
    TagQuery:Gameplay Tag Query Structure=Get ForceVisibleGameplayTagQuery() -> (ForceVisibleGameplayTagQuery:Gameplay Tag Query Structure)
    TriggerCondition:EWaitGameplayTagQueryTriggerCondition Enum=WhenTrue
    bOnlyTriggerOnce:bool=false
   ) -> (then:Exec, AsyncTaskProxy:Ability Async Wait Gameplay Tag Query Object Reference, Triggered:Exec) {is=title K2Node_AsyncAction}
   > n7 = WaitGameplayTagQueryOnActor(TargetActor=@n3, TagQuery=Get ForceVisibleGameplayTagQuery(), TriggerCondition=WhenFalse, bOnlyTriggerOnce=false)
   SetIsForceVisible!(execute=@n7.Triggered, self, NewIsForceVisible:bool=false) -> (then)
   SetIsForceVisible(execute=@n6.Triggered, NewIsForceVisible=true)
   Set IsRiding(execute=@n5.Removed, IsRiding:bool=false) -> (Output_Get:bool)
   > UpdateVisibilityFromState()
   Set IsRiding(execute=@n4.Added, IsRiding=true)
   > UpdateVisibilityFromState()
   n8 = OnCurrentEnergyChanged!() -> (OutputDelegate, then, EnergyComponent:Fort Component Energy Object Reference, PreviousValue:float=~) {is=title K2Node_CustomEvent}
   > n9 = Set CachedCurrentEnergy(CachedCurrentEnergy:double=GetCurrentEnergyPercentage!(self:Fort Component Energy Object Reference=@n8.EnergyComponent) -> (ReturnValue:float)) -> (Output_Get:double)
   > SetStaminaProgress(CurrentEnergy=@n9)
   > n10 = Branch(Condition="float >= float"(A=@n9, B=1))
   > HideSprintMeter()
   ShowSprintMeter(execute=@n10.else)
   > n11 = Branch(Condition=Get HasIntroPlayed() -> (HasIntroPlayed:bool))
   Branch(
    execute=[@n11.then, @n12.then]
    Condition="float <= float"(A=Get CachedCurrentEnergy() -> (CachedCurrentEnergy:double), B=Get Low Energy Threshold() -> (Low Energy Threshold:double)) -> (ReturnValue=~)
   )
   > TryPlayingLowState()
   PlayAnimation(execute=@n11.else, InAnimation=Get Anim_Intro_CenterScreenVersion() -> (Anim_Intro_CenterScreenVersion:Widget Animation Object Reference))
   > n12 = Set HasIntroPlayed(HasIntroPlayed:bool=true) -> (Output_Get:bool)
   HideSprintMeter() {is=title K2Node_CustomEvent}
   n13 = Get IsForceVisible() -> (IsForceVisible:bool)
   n14 = IsAnimationPlaying(InAnimation=Get AnimFullBar() -> (AnimFullBar:Widget Animation Object Reference))
   n15 = Get CachedCurrentEnergy()
   > Branch(Condition=AND Boolean!(A=NOT Boolean(A=@n13), B="float >= float"(A=@n15, B=1), C:bool=NOT Boolean(A=@n14)) -> (ReturnValue))
   > Set IsSprintMeterVisibile(IsSprintMeterVisibile:bool=false) -> (Output_Get:bool)
   > n16 = "Play Animation with Finished event (legacy)"!(
    execute:Exec
    Widget:User Widget Object Reference=Self-Reference()
    InAnimation:Widget Animation Object Reference=Get AnimFullBar()
    StartAtTime:float=0
    NumLoopsToPlay:int=1
    PlayMode:EUMGSequencePlayMode Enum
    PlaybackSpeed:float=1
   ) -> (then:Exec, Result:UMGSequence Player Object Reference, Finished:Exec) {is=title K2Node_PlayAnimation}
   n17 = Branch(execute=@n16.Finished, Condition=Get IsSprintMeterVisibile() -> (IsSprintMeterVisibile:bool))
   SetVisibility(execute=@n17.else, self=Get Overlay_Energy(), InVisibility=Hidden)
   > Set HasIntroPlayed(HasIntroPlayed=false)
   ShowSprintMeter() {is=title K2Node_CustomEvent}
   > Set IsSprintMeterVisibile(IsSprintMeterVisibile=true)
   > SetBracketLowThresholdVisuals!(execute, self, LowThreshold:float=Get Low Energy Threshold()) -> (then)
   > n18 = Sequence()
   > Branch(Condition=IsAnimationPlaying(InAnimation=Get AnimFullBar()))
   > StopAnimation(InAnimation=Get AnimFullBar())
   SetVisibility(execute=@n18.then_1, self=Get Overlay_Energy(), InVisibility=HitTestInvisible)
   > n19 = Branch(Condition=IsAnimationPlaying(InAnimation=Get AnimLow() -> (AnimLow:Widget Animation Object Reference)))
   SetRenderOpacity(execute=@n19.else, self=Get Overlay_Energy(), InOpacity=1)
   Event Destruct()
   > UnregisterWidgetHost!(
    execute
    WidgetHost:Widget Object Reference
    WidgetHostTag:"Gameplay Tag Structure (by ref)"=Get SprintMeterSlotTag() -> (SprintMeterSlotTag:Gameplay Tag Structure)
   ) -> (then)
   > RemoveBinds()
   n20 = Create Event()
   n21 = Get FortHUDContext()
   > Unbind Event from On Pawn Set(self:Fort HUDContext Object Reference=@n21, Delegate=@n20) {is=title K2Node_RemoveDelegate}
   n22 = Event Add Content To Host Widget!() -> (
    OutputDelegate
    then
    WidgetHostTag:"Gameplay Tag Structure (by ref)"
    ContentWidget:Widget Object Reference
    Padding:"Margin Structure (by ref)"
    bFill:bool
   ) {is=title K2Node_Event}
   > Branch(Condition=MatchesTag!(TagOne:Gameplay Tag Structure=@n22.WidgetHostTag, TagTwo:Gameplay Tag Structure=Get SprintMeterSlotTag(), bExactMatch:bool=true) -> (ReturnValue:bool))
   n23 = Get BarSwitcher() -> (BarSwitcher:Common Activatable Widget Switcher Object Reference)
   > AddChild(self=@n23, Content=@n22.ContentWidget)
   > SetActiveWidget(self:Widget Switcher Object Reference=@n23, Widget=@n22.ContentWidget)
   Event Construct()
   > RegisterWidgetHost!(
    execute
    WidgetHost:Widget Object Reference
    WidgetHostTag:"Gameplay Tag Structure (by ref)"=Get SprintMeterSlotTag()
    IsPlayerSpecific:bool=true
   ) -> (then)
   > SetVisibility(self=Get Overlay_Energy(), InVisibility=Hidden)
   > Bind Event to On Pawn Set(self:Fort HUDContext Object Reference=@n21, Delegate=@n20) {is=title K2Node_AddDelegate}
   > SetupEventBinds() {is=title K2Node_CallFunction}
  ]}
 FadeOutHighlightBar {nodes [
   FadeOutHighlightBar() {is=title K2Node_FunctionEntry}
   > Clear and Invalidate Timer by Handle(Handle=Get StartHighlightFadeOut_Handle())
   > Clear and Invalidate Timer by Handle(Handle=Get EmptyHighlight_Handle())
   > n1 = Set Timer by Event(Delegate=Create Event(), Time=0.03, bLooping=true, bMaxOncePerFrame=true, InitialStartDelay=0.03)
   > Set EmptyHighlight_Handle(EmptyHighlight_Handle:Timer Handle Structure=@n1) -> (Output_Get:Timer Handle Structure)
   > Set EmptyHighlight_StartTime(EmptyHighlight_StartTime:double=GetTimeSeconds()) -> (Output_Get:double)
  ]}
 GetIsDBNO {nodes [
   GetIsDBNO() {is=title K2Node_FunctionEntry}
   > n1 = GetPawnFromPlayerStateZone(InActor=GetPlayerStateZone(self=Get FortHUDContext()))
   > n2 = Branch(Condition=@n1.PawnReturned)
   n3 = AreDBNOEnhancementsEnabled(Pawn=@n1)
   > Return Node!(
    execute
    IsDBNO:bool=AND Boolean!(A=IsDBNO!(self:Fort Pawn Object Reference=@n1) -> (ReturnValue:bool), B=NOT Boolean(A=@n3)) -> (ReturnValue)
   )
   Return Node(execute=@n2.else, IsDBNO=false)
  ]}
 GetIsInVehicle {nodes [
   GetIsInVehicle() {is=title K2Node_FunctionEntry}
   > n1 = Cast To FortPlayerController(Object=GetOwningPlayer())
   n2 = GetPlayerPawn!(self:Fort Player Controller Object Reference=@n1) -> (ReturnValue:Fort Player Pawn Object Reference)
   > n3 = Is Valid(InputObject=@n2)
   Return Node!(execute=[@n3.Is Not Valid, @n1.CastFailed], IsInVehicle:bool=false)
   > Return Node(IsInVehicle=IsInVehicle!(self:Fort Player Pawn Object Reference=@n2) -> (ReturnValue:bool))
  ]}
 GetIsQuickbarEquipped {nodes [
   GetIsQuickbarEquipped() {is=title K2Node_FunctionEntry}
   > n1 = Cast To FortPlayerControllerGameplay!(execute, Object=GetOwningPlayer()) -> (then, CastFailed, AsFort Player Controller Gameplay:Fort Player Controller Gameplay Object Reference) {is=title K2Node_DynamicCast}
   > Return Node(IsInVehicle=IsCreativeQuickbarActive!(self:Fort Player Controller Gameplay Object Reference=@n1) -> (ReturnValue:bool))
   Return Node(execute=@n1.CastFailed, IsInVehicle=false)
  ]}
 HandleEnterVehicle {nodes [
   HandleEnterVehicle() {is=title K2Node_FunctionEntry}
   > UpdateVisibilityFromState()
  ]}
 HandleExitVehicle {nodes [
   HandleExitVehicle() {is=title K2Node_FunctionEntry}
   > UpdateVisibilityFromState()
  ]}
 OnCreativeQuickbarEquippedChanged {nodes [
   OnCreativeQuickbarEquippedChanged!() -> (then, bIsQuickbarEquipped:bool=~) {is=title K2Node_FunctionEntry}
   > UpdateVisibilityFromState()
  ]}
 OnDBNOStateChanged {nodes [
   n1 = OnDBNOStateChanged!() -> (then, FortPawn:Fort Pawn Object Reference, bInIsDBNO:bool=~) {is=title K2Node_FunctionEntry}
   > n2 = Get!(execute:Exec) -> (then:Exec, else:Exec, FortPawn:Fort Pawn Object Reference)
   n3 = NOT Boolean(A=AreDBNOEnhancementsEnabled(Pawn=@n2))
   > SetVisibility(InVisibility=Select(Option 0=HitTestInvisible, Option 1=Hidden, Index=AND Boolean(A=@n1.bInIsDBNO, B=@n3)))
  ]}
 OnHighlightActivated {nodes [
   OnHighlightActivated!() -> (then, EnergyComponent:Fort Component Energy Object Reference, PreviousValue:float=~) {is=title K2Node_FunctionEntry}
   > Set IsGainHighlightActive(IsGainHighlightActive=true)
   > Set ValueBeforeHighlightedGain(ValueBeforeHighlightedGain=Get PrevEnergy())
   > n1 = Branch(Condition=Is Valid Timer Handle!(Handle:Timer Handle Structure=Get StartHighlightFadeOut_Handle()) -> (ReturnValue:bool))
   > n2 = Clear and Invalidate Timer by Handle(Handle=Get StartHighlightFadeOut_Handle())
   n3 = Set Timer by Event(execute=[@n2.then, @n1.else], Delegate=Create Event(), Time=Get Highlight_TotalSeconds() -> (Highlight_TotalSeconds:double), bLooping=false)
   > Set StartHighlightFadeOut_Handle(StartHighlightFadeOut_Handle:Timer Handle Structure=@n3) -> (Output_Get:Timer Handle Structure)
   > Set PrevEnergy(PrevEnergy:double=GetCurrentEnergyPercentage(self=Get EnergyComponent() -> (EnergyComponent:Fort Component Energy Object Reference))) -> (Output_Get:double)
   > Clear and Invalidate Timer by Handle(Handle=Get EmptyHighlight_Handle())
  ]}
 OnOvershieldVisibilityChanged {nodes [
   OnOvershieldVisibilityChanged!() -> (then, NewVisible:bool=~) {is=title K2Node_FunctionEntry}
   > CheckOvershieldVisiblity() {is=title K2Node_CallFunction}
  ]}
 RemoveContentFromWidgetHost {nodes [
   n1 = Remove Content From Host Widget!() -> (
    then
    WidgetHostTag:"Gameplay Tag Structure (by ref)"
    ContentWidget:Widget Object Reference
    RemovalType:EWidgetSlottingRemovalType Enum=Unparent
   ) {is=title K2Node_FunctionEntry}
   > n2 = Branch(Condition=MatchesTag(TagOne=@n1, TagTwo=Get SprintMeterSlotTag(), bExactMatch=true))
   > n3 = RemoveChild(self=Get BarSwitcher(), Content=@n1.ContentWidget)
   > Return Node!(execute, ReturnValue:bool=@n3)
   Return Node(execute=@n2.else, ReturnValue=false)
  ]}
 SetBracketEmptyVisuals {nodes [
   n1 = SetBracketEmptyVisuals!() -> (then, Progress:float=~) {is=title K2Node_FunctionEntry}
   > n2 = For Each Loop(Array:Array of Image Object References=Make Array!(
     "[0]":Image Object Reference=Get Image_Bracket() -> (Image_Bracket:Image Object Reference)
     "[1]":Image Object Reference=Get Image_Bracket_Rght() -> (Image_Bracket_Rght:Image Object Reference)
    ) -> (Array:Array of Image Object References)) -> (Array Element:Image Object Reference)
   > n3 = GetDynamicMaterial(self=@n2)
   > SetScalarParameterValue(self=@n3, ParameterName=Progress, Value=@n1)
  ]}
 SetBracketLowThresholdVisuals {nodes [
   n1 = SetBracketLowThresholdVisuals!() -> (then, LowThreshold:float=~) {is=title K2Node_FunctionEntry}
   > n2 = For Each Loop(Array=Make Array("[0]"=Get Image_Bracket(), "[1]"=Get Image_Bracket_Rght()))
   > n3 = GetDynamicMaterial(self=@n2)
   > SetScalarParameterValue(self=@n3, ParameterName=LowTheshold, Value=@n1)
  ]}
 SetDelta {nodes [
   n1 = SetDelta!() -> (then, CurrentEnergy:double=~) {is=title K2Node_FunctionEntry}
   n2 = Get Anim_DeltaThreshold() -> (Anim_DeltaThreshold:double)
   n3 = Get PrevEnergy()
   > n4 = Branch(Condition="float <= float"(A=@n1, B=float - float(A=@n3, B=@n2)) -> (ReturnValue=~))
   > n5 = GetDynamicMaterial(self=Get Image_StaminaBar())
   > SetScalarParameterValue(self=@n5, ParameterName=DeltaProgress, Value=Get PrevEnergy())
   > Clear and Invalidate Timer by Handle(Handle=Get EmptyDelta_Handle() -> (EmptyDelta_Handle:Timer Handle Structure))
   > n6 = Set Timer by Event(Delegate=Create Event(), Time=0.33, bLooping=false)
   > Set EmptyDelta_Handle(EmptyDelta_Handle:Timer Handle Structure=@n6) -> (Output_Get:Timer Handle Structure)
   n7 = Branch(execute=@n4.else, Condition=Is Timer Active by Handle!(Handle:Timer Handle Structure=Get EmptyDelta_Handle()) -> (ReturnValue:bool))
   EmptyDelta(execute=@n7.else) {is=title K2Node_CallFunction}
  ]}
 SetIsForceVisible {nodes [
   n1 = SetIsForceVisible!() -> (then, NewIsForceVisible:bool=~) {is=title K2Node_FunctionEntry}
   > n2 = Set IsForceVisible(IsForceVisible:bool=@n1) -> (Output_Get:bool)
   > n3 = Branch(Condition=@n2)
   > ShowSprintMeter() {is=title K2Node_CallFunction}
   HideSprintMeter(execute=@n3.else) {is=title K2Node_CallFunction}
  ]}
 SetStaminaProgress {nodes [
   SetStaminaProgress!() -> (then, CurrentEnergy:double=~) {is=title K2Node_FunctionEntry}
   n1 = Get CurrentEnergy() -> (CurrentEnergy:double)
   > SetRenderOpacity(
    self=Get Image_SprintIcon() -> (Image_SprintIcon:Image Object Reference)
    InOpacity=Select(Option 0:double=1, Option 1:double=0.33, Index="float <= float"(A=@n1) -> (ReturnValue=~)) -> (ReturnValue:double)
   )
   > n2 = Branch(Condition=Get IsGainHighlightActive() -> (IsGainHighlightActive:bool))
   n3 = float - float(A=Get CurrentEnergy(), B=Get PrevEnergy())
   > n4 = Branch(Condition="float > float"(A=@n3))
   > n5 = Set ValueBeforeHighlightedGain(ValueBeforeHighlightedGain=float + float(A=@n3, B=Get ValueBeforeHighlightedGain()))
   n6 = Branch(execute=[@n5.then, @n4.else], Condition="float > float"(A=Get CurrentEnergy(), B=Get ValueBeforeHighlightedGain()))
   > n7 = GetDynamicMaterial(self=Get Image_StaminaBar())
   n8 = Get ValueBeforeHighlightedGain()
   > SetScalarParameterValue(self=@n7, ParameterName=Progress, Value=@n8)
   > SetScalarParameterValue(self=@n7, ParameterName=HighlightProgress, Value=Get CurrentEnergy())
   > SetBracketEmptyVisuals!(execute, self, Progress:float=@n8) -> (then) {is=title K2Node_CallFunction}
   > n9 = GetDynamicMaterial(self=Get Image_StaminaBar_Glow() -> (Image_StaminaBar_Glow:Image Object Reference))
   > SetScalarParameterValue(self=@n9, ParameterName=Progress, Value=@n8)
   > SetDelta!(execute, self, CurrentEnergy:double=Get ValueBeforeHighlightedGain()) -> (then) {is=title K2Node_CallFunction}
   > Set PrevEnergy(PrevEnergy=Get CurrentEnergy())
   DisableHighlightBar(execute=@n6.else)
   n10 = GetDynamicMaterial(execute=@n2.else, self=Get Image_StaminaBar())
   n11 = Get CurrentEnergy()
   > SetScalarParameterValue(self=@n10, ParameterName=Progress, Value=@n11)
   > SetBracketEmptyVisuals(Progress=@n11)
   > n12 = GetDynamicMaterial(self=Get Image_StaminaBar_Glow())
   > SetScalarParameterValue(self=@n12, ParameterName=Progress, Value=@n11)
   n13 = Get CurrentEnergy()
   > SetDelta(CurrentEnergy=@n13)
   > Set PrevEnergy(PrevEnergy=@n13)
  ]}
 TryPlayingLowState {nodes [
   TryPlayingLowState() {is=title K2Node_FunctionEntry}
   > n1 = Branch(Condition=IsAnimationPlaying(InAnimation=Get AnimLow()))
   PlayAnimation(execute=@n1.else, InAnimation=Get AnimLow())
  ]}
 UpdateVisibilityFromState {nodes [
   UpdateVisibilityFromState() {is=title K2Node_FunctionEntry}
   n1 = GetIsInVehicle!(self) -> (IsInVehicle:bool) {is=title K2Node_CallFunction}
   n2 = GetIsQuickbarEquipped!(self) -> (IsInVehicle:bool) {is=title K2Node_CallFunction}
   n3 = GetIsDBNO!(self) -> (IsDBNO:bool) {is=title K2Node_CallFunction}
   n4 = Get IsRiding() -> (IsRiding:bool)
   > SetVisibility(InVisibility=Select(
     Option 0:ESlateVisibility Enum=SelfHitTestInvisible
     Option 1:ESlateVisibility Enum=Hidden
     Index=OR Boolean!(A=@n4, B=@n1, C:bool=@n2, D:bool=@n3) -> (ReturnValue)
    ) -> (ReturnValue:ESlateVisibility Enum))
  ]}
]
variables [
 Anim_DeltaThreshold {type=number, default=0.05}
 BoundEnergyComponents {type="array<FortComponent_Energy>", default []}
 CachedCurrentEnergy {type=number}
 EmptyDelta_Handle {type=TimerHandle}
 EmptyHighlight_Handle {type=TimerHandle}
 EmptyHighlight_StartTime {type=number, default=0.3}
 EmptyHighlight_TotalSeconds {type=number, default=0.5}
 ForceVisibleGameplayTagQuery {
  type=GameplayTagQuery
  default {
   autoDescription=" ANY( Granted.Athena.Energy.ConstantRegen )"
   tokenStreamVersion=0
   userDescription=""
   queryTokenStream [0,1,1,1,0]
   tagDictionary [{tagName=Granted.Athena.Energy.ConstantRegen}]
  }
 }
 HasIntroPlayed {type=boolean}
 Highlight_TotalSeconds {type=number, default=1.5}
 HighlightGainTag {type=GameplayTag, default {tagName=HUD.Athena.TacticalSprint.HighlightGain}}
 IsForceVisible {type=boolean}
 IsGainHighlightActive {type=boolean}
 IsRiding {type=boolean}
 IsSprintMeterVisibile {type=boolean}
 Low Energy Threshold {type=number, default=0.25}
 PrevEnergy {type=number}
 SprintMeterSlotTag {type=GameplayTag, default {tagName=DynamicUI.HUD.Fort.Athena.PlayerHitPointInfo.SprintMeter.Slot}}
 StartHighlightFadeOut_Handle {type=TimerHandle}
 ValueBeforeHighlightedGain {type=number}
]
widgets {
 info {inheritedWidgetCount=1, parentClass=&FortHUDElementWidget, rootWidgetClass=&Overlay, widgetCount=12}
 widgets [
  Overlay_Energy {widgetClassPath=&Overlay, bIsVariable=true, widget=Overlay_Energy}
  HorizontalBox_0 {
   widgetClassPath=&HorizontalBox
   parent=Overlay_Energy
   widget=HorizontalBox_0
   slot {class=OverlaySlot, properties {verticalAlignment=VAlign_Center}}
  }
  Image_Bracket {
   widgetClassPath=&Image
   bIsVariable=true
   parent=HorizontalBox_0
   widget=Image_Bracket
   properties {
    visibility=SelfHitTestInvisible
    brush {
     bIsDynamicallyLoaded=false
     drawAs=Image
     imageType=NoImage
     mirroring=NoMirror
     resourceName=None
     resourceObject=/BattleRoyaleGameUI/Shared/FrameDecor/M_UI_BarBracket
     tiling=NoTile
     imageSize=<28 24>
     margin=<ltrb 0 0 0 0>
     outlineSettings {
      bUseBrushTransparency=false
      roundingType=HalfHeightRadius
      width=0
      color {colorUseRule=UseColor_Specified, specifiedColor=<rgba 0 0 0 0>}
      cornerRadii=<0 0 0 0>
     }
     tintColor {colorUseRule=UseColor_Specified, specifiedColor=<rgba 1 1 1 1>}
     uVRegion {bIsValid=false, max=<0 0>, min=<0 0>}
    }
    renderTransform {angle=0, scale=<1.25 1.75>, shear=<0 0>, translation=<0 0>}
   }
   slot {class=HorizontalBoxSlot, properties {horizontalAlignment=HAlign_Left, verticalAlignment=VAlign_Bottom, padding=<ltrb 0 0 -2 0>}}
  }
  VerticalBox {
   widgetClassPath=&VerticalBox
   parent=HorizontalBox_0
   widget=VerticalBox
   properties {bOverride_Cursor=true}
   slot {class=HorizontalBoxSlot, properties {verticalAlignment=VAlign_Center, padding=<ltrb -16 0 -16 0>}}
  }
  Image_SprintIcon {
   widgetClassPath=&Image
   bIsVariable=true
   parent=VerticalBox
   widget=Image_SprintIcon
   properties {
    visibility=Collapsed
    brush {
     bIsDynamicallyLoaded=false
     drawAs=Image
     imageType=FullColor
     mirroring=NoMirror
     resourceName=None
     resourceObject=/BattleRoyaleGameUI/HUD/PlayerHitPointInfo/Assets/M_UI_SprintIcon
     tiling=NoTile
     imageSize=<32 32>
     margin=<ltrb 0 0 0 0>
     outlineSettings {
      bUseBrushTransparency=false
      roundingType=HalfHeightRadius
      width=0
      color {colorUseRule=UseColor_Specified, specifiedColor=<rgba 0 0 0 0>}
      cornerRadii=<0 0 0 0>
     }
     tintColor {colorUseRule=UseColor_Specified, specifiedColor=<rgba 1 1 1 1>}
     uVRegion {bIsValid=false, max=<0 0>, min=<0 0>}
    }
    renderTransform {angle=0, scale=<1.5 1.5>, shear=<0 0>, translation=<0 6>}
   }
   slot {class=VerticalBoxSlot, properties {verticalAlignment=VAlign_Top}}
  }
  SizeBox_0 {
   widgetClassPath=&SizeBox
   parent=VerticalBox
   widget=SizeBox_0
   properties {bOverride_HeightOverride=true, bOverride_WidthOverride=true, heightOverride=8, widthOverride=208}
   slot {class=VerticalBoxSlot}
  }
  BarSwitcher {
   widgetClassPath=&CommonActivatableWidgetSwitcher
   bIsVariable=true
   parent=SizeBox_0
   widget=BarSwitcher
   slot {class=SizeBoxSlot}
  }
  DefaultBar {widgetClassPath=&Overlay, parent=BarSwitcher, widget=DefaultBar, slot {class=WidgetSwitcherSlot}}
  Image_StaminaBar_Glow {
   widgetClassPath=&Image
   bIsVariable=true
   parent=DefaultBar
   widget=Image_StaminaBar_Glow
   properties {
    visibility=SelfHitTestInvisible
    brush {
     bIsDynamicallyLoaded=false
     drawAs=Image
     imageType=FullColor
     mirroring=NoMirror
     resourceName=None
     resourceObject=/BattleRoyaleGameUI/HUD/PlayerHitPointInfo/Assets/MI_UI_StaminaBar_Glow
     tiling=NoTile
     imageSize=<32 32>
     margin=<ltrb 0 0 0 0>
     outlineSettings {
      bUseBrushTransparency=false
      roundingType=HalfHeightRadius
      width=0
      color {colorUseRule=UseColor_Specified, specifiedColor=<rgba 0 0 0 0>}
      cornerRadii=<0 0 0 0>
     }
     tintColor {colorUseRule=UseColor_Specified, specifiedColor=<rgba 1 1 1 1>}
     uVRegion {bIsValid=false, max=<0 0>, min=<0 0>}
    }
   }
   slot {class=OverlaySlot, properties {horizontalAlignment=HAlign_Fill, verticalAlignment=VAlign_Fill}}
  }
  Image_StaminaBar {
   widgetClassPath=&Image
   bIsVariable=true
   parent=DefaultBar
   widget=Image_StaminaBar
   properties {
    visibility=SelfHitTestInvisible
    brush {
     bIsDynamicallyLoaded=false
     drawAs=Image
     imageType=FullColor
     mirroring=NoMirror
     resourceName=None
     resourceObject=/BattleRoyaleGameUI/HUD/PlayerHitPointInfo/Assets/M_UI_StaminaBar
     tiling=NoTile
     imageSize=<32 32>
     margin=<ltrb 0.5 0.5 0.5 0.5>
     outlineSettings {
      bUseBrushTransparency=false
      roundingType=HalfHeightRadius
      width=0
      color {colorUseRule=UseColor_Specified, specifiedColor=<rgba 0 0 0 0>}
      cornerRadii=<0 0 0 0>
     }
     tintColor {colorUseRule=UseColor_Specified, specifiedColor=<rgba 1 1 1 1>}
     uVRegion {bIsValid=false, max=<0 0>, min=<0 0>}
    }
   }
   slot {class=OverlaySlot, properties {horizontalAlignment=HAlign_Fill, verticalAlignment=VAlign_Fill}}
  }
  Image_Bracket_Rght {
   widgetClassPath=&Image
   bIsVariable=true
   parent=HorizontalBox_0
   widget=Image_Bracket_Rght
   properties {
    visibility=SelfHitTestInvisible
    brush {
     bIsDynamicallyLoaded=false
     drawAs=Image
     imageType=NoImage
     mirroring=NoMirror
     resourceName=None
     resourceObject=&M_UI_BarBracket
     tiling=NoTile
     imageSize=<28 24>
     margin=<ltrb 0 0 0 0>
     outlineSettings {
      bUseBrushTransparency=false
      roundingType=HalfHeightRadius
      width=0
      color {colorUseRule=UseColor_Specified, specifiedColor=<rgba 0 0 0 0>}
      cornerRadii=<0 0 0 0>
     }
     tintColor {colorUseRule=UseColor_Specified, specifiedColor=<rgba 1 1 1 1>}
     uVRegion {bIsValid=false, max=<0 0>, min=<0 0>}
    }
    renderTransform {angle=0, scale=<-1.25 1.75>, shear=<0 0>, translation=<0 0>}
   }
   slot {class=HorizontalBoxSlot, properties {horizontalAlignment=HAlign_Right, verticalAlignment=VAlign_Bottom, padding=<ltrb -2 0 0 0>}}
  }
  TutorialHighlight {widgetClassPath=&TutorialHighlight}
 ]
}
