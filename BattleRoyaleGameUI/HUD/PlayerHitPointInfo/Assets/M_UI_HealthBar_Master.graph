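# Translucent UI master material that draws the player HUD health bar, with states for down-but-not-out and for shield or vehicle damage.
# Progress cuts the gradient fill, DamageProgress masks the damage segment, and marker, glow, shadow and optional separator masks layer over them.
# DBNO swaps colors and adds a sine-timed flash, while Anim_DMGGlow drives damage glow.
# Vertex padding, SDF shapes and low-end value switchers shape the rendering.
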Material /BattleRoyaleGameUI/HUD/PlayerHitPointInfo/Assets/M_UI_HealthBar_Master
graph {
 properties {blendMode=BLEND_Translucent, materialDomain=MD_UI, pixelDepthOffsetMode=PDOM_Legacy, usedShadingModels=Default Lit}
 nodes [
  n1 = ScalarParameter() {group=DBNO, parameterName=IsDBNO, sortPriority=0}
  n2 = VectorParameter() {group=DBNO, parameterName=DBNOFlashingColor, sortPriority=4, defaultValue=<rgba 0.0194 0.0024 0.006 1>}
  n3 = Time() {bIgnorePause=true, bOverride_Period=true, period=1}
  n4 = Multiply(A=@n2.RGB, B=ConstantBiasScale(None=Sine(None=@n3)))
  n5 = LinearInterpolate(A=NamedRerouteUsage(), B=NamedRerouteUsage(), Alpha=NamedRerouteUsage())
  n6 = LinearInterpolate(A=NamedRerouteUsage(), B=NamedRerouteUsage(), Alpha=NamedRerouteUsage())
  n7 = NamedRerouteUsage()
  LinearInterpolate(A=NamedRerouteUsage(), B=Add(A=Multiply(A=@n1, B=@n4), B=LinearInterpolate(A=@n5, B=@n6, Alpha=@n7)), Alpha=NamedRerouteUsage())
  n8 = NamedRerouteUsage()
  n9 = LinearInterpolate(Alpha=NamedRerouteUsage()) {constA=0.5}
  n10 = LinearInterpolate(A=Constant() {r=0.5}, Alpha=ScalarParameter() {parameterName=Anim_DMGGlow}) {constB=3}
  n11 = LinearInterpolate(A=Multiply(A=@n10) {constB=0.6667}, B=@n10, Alpha=NamedRerouteUsage()) {constA=0.5}
  n12 = NamedRerouteUsage()
  n13 = Constant() {r=0.85}
  n14 = Constant() {r=0.525}
  n15 = MaterialFunctionCall!(Default=ScalarParameter() {defaultValue=0.75, group=Color, parameterName=BackgroundOpacity, sortPriority=7}, LowEnd=Constant() {r=0.464}) {materialFunction=/BattleRoyaleGameUI/Shared/MF_UI_LowEnd_ValueSwitcher}
  n16 = NamedRerouteUsage()
  n17 = LinearInterpolate(
   A=LinearInterpolate(
    A=MaterialFunctionCall(Default=@n13, LowEnd=@n14) {materialFunction=&MF_UI_LowEnd_ValueSwitcher}.Result
    B=@n15.Result
    Alpha=MaterialFunctionCall("Value (0-1)"=@n16) {materialFunction=&MF_UI_EaseCurves}.EaseInQuint
   ) {constA=0.85}
   B=@n15.Result
   Alpha=OneMinus(None=NamedRerouteUsage())
  )
  n18 = NamedRerouteUsage()
  n19 = Saturate(None=OneMinus(None=NamedRerouteUsage()))
  n20 = NamedRerouteDeclaration(None=LinearInterpolate(Alpha=@n19) {constA=0.85}) {name=FadeMaskController, nodeColor=<rgba 0 0 0 1>}
  n21 = NamedRerouteUsage()
  n22 = Saturate(None=Subtract(A=NamedRerouteUsage(), B=NamedRerouteUsage()))
  n23 = NamedRerouteUsage()
  n24 = NamedRerouteUsage()
  n25 = LinearInterpolate(
   A=MaterialFunctionCall(Default=Constant() {r=0.125}, LowEnd=Constant() {r=0.1667}) {materialFunction=&MF_UI_LowEnd_ValueSwitcher}.Result
   B=MaterialFunctionCall(Default=Constant() {r=0.3333}, LowEnd=Constant() {r=0.268}) {materialFunction=&MF_UI_LowEnd_ValueSwitcher}.Result
   Alpha=@n19
  ) {constA=0.25}
  n26 = NamedRerouteUsage()
  n27 = LinearInterpolate(A=@n25, B=Multiply(A=Saturate(None=@n26), B=@n25))
  n28 = MaterialFunctionCall(Default=Constant() {r=0.5}, LowEnd=Constant() {r=0.3333}) {materialFunction=&MF_UI_LowEnd_ValueSwitcher}
  n29 = LinearInterpolate(
   A=LinearInterpolate(A=Subtract(A=@n17, B=@n18), B=LinearInterpolate(A=@n20, Alpha=@n21), Alpha=LinearInterpolate(A=@n22, B=@n23, Alpha=@n24))
   B=LinearInterpolate(B=LinearInterpolate(A=@n27, B=@n28.Result, Alpha=@n21) {constB=0.5}, Alpha=NamedRerouteUsage())
   Alpha=NamedRerouteUsage()
  ) {constB=0.25}
  n30 = NamedRerouteUsage()
  n31 = NamedRerouteUsage()
  n32 = Subtract(A=NamedRerouteUsage(), B=LinearInterpolate(A=NamedRerouteUsage(), Alpha=Subtract(A=@n30, B=@n31)) {constB=0})
  LinearInterpolate(
   A=Multiply(A=Multiply(A=@n8, B=@n9), B=LinearInterpolate(A=@n11, Alpha=@n12) {constA=0.5, constB=0}) {constB=0.5}
   Alpha=Multiply(A=LinearInterpolate(A=@n29, Alpha=@n32), B=NamedRerouteUsage())
  )
  n33 = NamedRerouteUsage()
  n34 = NamedRerouteUsage()
  n35 = Saturate(None=SmoothStep(Max=@n33, Value=@n34))
  NamedRerouteDeclaration(None=@n35) {name=BarBottomShadowMask, nodeColor=<rgba 0 0 0 1>}
  n36 = NamedRerouteUsage()
  n37 = ComponentMask(None=@n36) {g=true}
  NamedRerouteDeclaration(None=@n37) {name=VerticalMask, nodeColor=<rgba 0 0 0 1>}
  n38 = Subtract(A=ComponentMask(None=MaterialFunctionCall() {materialFunction=&GetUserInterfaceUV}.Normalized UV) {g=true}) {constB=0.5}
  n39 = ScalarParameter() {defaultValue=0.025, group=Marker, parameterName=MarkerAngle, sortPriority=0}
  n40 = Abs(None=Add(A=NamedRerouteUsage(), B=Multiply(A=@n38, B=@n39)))
  n41 = ScalarParameter() {defaultValue=0.025, group=Marker, parameterName=MarkerWidth, sortPriority=1}
  n42 = ScalarParameter() {defaultValue=0.005, parameterName=ShadowWidth}
  n43 = Multiply(
   A=MaterialFunctionCall(
    DistanceField=@n40
    Threshold=Add(A=@n41, B=@n42)
    Smoothing=ScalarParameter() {defaultValue=0.0125, group=Marker, parameterName=MarkerShadowSmoothing, sortPriority=2}
   ) {materialFunction=&MF_UI_AntiAliasing}.Inner
   B=ScalarParameter() {defaultValue=0.125, parameterName=MarkerShadowOpacity}
  )
  n44 = NamedRerouteDeclaration(None=MaterialFunctionCall(DistanceField=@n40, Threshold=@n41, Smoothing=ScalarParameter() {group=Marker, parameterName=MarkerSmoothing, sortPriority=2}) {materialFunction=&MF_UI_AntiAliasing}.Inner) {name=MarkerMask, nodeColor=<rgba 0 0 0 1>}
  NamedRerouteDeclaration(None=Saturate(None=Subtract(A=@n43, B=@n44))) {name=MarkerShadow, nodeColor=<rgba 0 0 0 1>}
  n45 = Max(A=ScalarParameter() {group=Damage, parameterName=DamageProgress, sortPriority=0}, B=NamedRerouteUsage())
  n46 = Subtract(A=ComponentMask(None=NamedRerouteUsage()) {r=true}, B=@n45)
  n47 = OneMinus(None=Saturate(None=Ceil(None=@n46)))
  n48 = MaterialFunctionCall!(A=@n45, B=Constant()) {materialFunction=/FortUILibrary/MaterialFunctions/Basics/MF_UI_BooleanOperations}
  n49 = Saturate(None=NamedRerouteDeclaration(None=LinearInterpolate(A=@n47, Alpha=@n48.IsLessOrEqual) {constB=0}) {name=DamageMask, nodeColor=<rgba 0 0 0 1>})
  n50 = LinearInterpolate(
   A=CollectionParameter() {collection=/BattleRoyaleGameUI/Shared/MPC_BRGameUI, parameterName=Red}
   B=CollectionParameter() {collection=&MPC_BRGameUI, parameterName=Cyan}
   Alpha=ScalarParameter() {parameterName=IsShieldDMG}
  )
  n51 = NamedRerouteDeclaration(None=LinearInterpolate(
    A=ComponentMask(None=@n50) {b=true, g=true, r=true}
    B=VectorParameter() {group=Color, parameterName=DamageColor, sortPriority=4, defaultValue=<rgba 1 0.013 0.0931 1>}.RGB
    Alpha=ScalarParameter() {parameterName=IsVehicleDMG}
   )) {name=DMGColor, nodeColor=<rgba 0 0 0 1>}
  NamedRerouteDeclaration(None=Saturate(None=Multiply(A=@n49, B=@n51))) {name=DamageColor, nodeColor=<rgba 0 0 0 1>}
  n52 = SmoothStep(Min=Add(A=@n33, B=ScalarParameter() {defaultValue=0.1, parameterName=ProgressFill_DropShadowSize}), Max=@n33, Value=@n34)
  NamedRerouteDeclaration(None=MaterialFunctionCall("Value (0-1)"=Saturate(None=@n52)) {materialFunction=&MF_UI_EaseCurves}.EaseInQuint) {name=ProgressFillDropShadowMask, nodeColor=<rgba 0 0 0 1>}
  n53 = NamedRerouteUsage()
  n54 = LinearInterpolate(A=OneMinus(None=ComponentMask(None=@n53) {r=true}), Alpha=NamedRerouteUsage())
  n55 = OneMinus(None=@n54)
  n56 = Add(
   A=MaterialFunctionCall() {materialFunction=&GetUserInterfaceUV}.Normalized UV
   B=AppendVector(A=Multiply(A=@n55) {constB=0.5}, B=Constant())
  )
  n57 = AppendVector(A=@n54, B=Constant() {r=1})
  n58 = Constant() {r=0.6}
  n59 = StaticBool() {value=true}
  n60 = NamedRerouteUsage()
  n61 = Saturate(None=MaterialFunctionCall(
    DistanceField=MaterialFunctionCall(UVs=@n56, Size=@n57, CornerRoundness=@n58, KeepAspectRatio=@n59, PixelSize=@n60) {materialFunction=&MF_UI_Box}.SDF
    Threshold=ScalarParameter() {defaultValue=-0.25, parameterName=Param_1}
    Smoothing=ScalarParameter() {defaultValue=1}
   ) {materialFunction=&MF_UI_AntiAliasing}.Inner)
  NamedRerouteDeclaration(None=LinearInterpolate(
    A=MaterialFunctionCall("Value (0-1)"=@n61) {materialFunction=&MF_UI_EaseCurves}.EaseInQuint
    Alpha=NamedRerouteUsage()
   ) {constB=0}) {name=ProgressGlowMask, nodeColor=<rgba 0 0 0 1>}
  n62 = ScalarParameter() {defaultValue=1, group=Fill, parameterName=Progress, sortPriority=0}
  NamedRerouteDeclaration(None=MaterialFunctionCall(A=@n62, B=Constant()) {materialFunction=&MF_UI_BooleanOperations}.IsLessOrEqual) {name=ProgressEqualsZeroLogic, nodeColor=<rgba 0 0 0 1>}
  n63 = Constant2Vector() {g=1}
  n64 = Constant2Vector() {g=-0.5, r=0.5}
  n65 = MaterialFunctionCall(
   Input=ScalarParameter() {defaultValue=0.8, group=Separator, parameterName=SeparatorGapPosition}
   Input Low=@n63.R
   Input High=@n63.R
   Target Low=@n64.R
   Target High=@n64.R
  ) {materialFunction=&RemapValueRange}
  n66 = Constant()
  n67 = MaterialFunctionCall!(
   UVs=Add(A=NamedRerouteUsage(), B=AppendVector(A=@n65.Result, B=@n66)) {constB=0}
   Size=AppendVector(A=ScalarParameter() {defaultValue=0.0125, group=Separator, parameterName=SeparatorGapWIdth}, B=Constant() {r=1.5})
   CornerRoundness
   KeepAspectRatio
   PixelSize
  ) {materialFunction=/FortUILibrary/MaterialFunctions/Shapes/SDFs/MF_SDF_Box}
  NamedRerouteDeclaration(None=Multiply(
    A=MaterialFunctionCall(DistanceField=@n67.SDF) {materialFunction=&MF_UI_AntiAliasing}.Inner
    B=ScalarParameter() {group=Separator, parameterName=ShowSeparatorGap}
   )) {name=Separator, nodeColor=<rgba 0.8984 1 0 1>}
  NamedRerouteDeclaration(None=MaterialFunctionCall(A=@n62, B=Constant() {r=1}) {materialFunction=&MF_UI_BooleanOperations}.IsEqual) {name=ProgressEqualsFullLogic, nodeColor=<rgba 0 0 0 1>}
  n68 = MaterialFunctionCall() {materialFunction=&GetUserInterfaceUV}
  n69 = MaterialFunctionCall!(Anchor, Translation, Padding=Multiply(A=ComponentMask(None=@n68.Pixel Size) {g=true}) {constB=0.75}, Scale=Constant2Vector() {g=1, r=1}) {materialFunction=/FortUILibrary/MaterialFunctions/Basics/MF_UI_VertexTransform}
  NamedRerouteDeclaration(None=@n69.New Pixel Size) {name=NewPixelSize, nodeColor=<rgba 0 0 0 1>}
  NamedRerouteDeclaration(None=@n69.Total Scale Factor) {name=TotalScaleFactor, nodeColor=<rgba 0 0 0 1>}
  n70 = MaterialFunctionCall!(BufferSource) {materialFunction=/BattleRoyaleGameUI/Shared/MF_UI_AdaptiveColors}
  NamedRerouteDeclaration(None=MaterialFunctionCall(Default=@n70.Dim, LowEnd=@n70.Dim) {materialFunction=&MF_UI_LowEnd_ValueSwitcher}.Result) {name=Dim, nodeColor=<rgba 0 0 0 1>}
  NamedRerouteDeclaration(None=@n70.Dark) {name=Dark, nodeColor=<rgba 0 0 0 1>}
  n71 = Constant() {r=1}
  n72 = NamedRerouteUsage()
  n73 = MaterialFunctionCall(
   UVs=MaterialFunctionCall() {materialFunction=&GetUserInterfaceUV}.Normalized UV
   Texture Scale=NamedRerouteDeclaration(None=Divide(A=@n71, B=@n72)) {name=ScaleFactorRatio, nodeColor=<rgba 0 0 0 1>}
  ) {materialFunction=&ScaleUVsByCenter}
  NamedRerouteDeclaration(None=MaterialFunctionCall!(UVs=@n73.UVs, Alpha=Constant() {r=1}) {materialFunction=/FortUILibrary/MaterialFunctions/Basics/MF_UI_ClipEdgePixel}.Result) {name=BarMask, nodeColor=<rgba 0 0 0 1>}
  n74 = VectorParameter() {group=Color, parameterName=FillColorStart, sortPriority=0, defaultValue=<rgba 0.003 0.6038 0.0437 1>}
  n75 = NamedRerouteDeclaration(None=ComponentMask(None=@n36) {r=true}) {name=RHorizontalMask, nodeColor=<rgba 0 0 0 1>}
  n76 = VectorParameter() {group=DBNO, parameterName=DBNOFillColorStart, sortPriority=1, defaultValue=<rgba 0.3185 0.0086 0.0666 1>}
  n77 = NamedRerouteDeclaration(None=ComponentMask(None=CollectionParameter() {collection=&MPC_BRGameUI, parameterName=Red}) {b=true, g=true, r=true}) {name=Red, nodeColor=<rgba 0 0 0 1>}
  n78 = ScalarParameter() {parameterName=UseSquadHealthBarCriticalColor}
  n79 = VectorParameter() {group=DBNO, parameterName=DBNOFillColorEnd, sortPriority=2, defaultValue=<rgba 0.6038 0.0052 0.0908 1>}
  n80 = ScalarParameter() {group=DBNO, parameterName=IsDBNO, sortPriority=0}
  n81 = LinearInterpolate(
   A=LinearInterpolate(
    A=@n74.RGB
    B=VectorParameter() {group=Color, parameterName=FillColorEnd, sortPriority=1, defaultValue=<rgba 0.3223 0.75 0.0168 1>}.RGB
    Alpha=@n75
   )
   B=LinearInterpolate(A=LinearInterpolate(A=@n76.RGB, B=@n77, Alpha=@n78), B=@n79.RGB, Alpha=@n75)
   Alpha=@n80
  )
  n82 = LinearInterpolate(A=@n81, B=LinearInterpolate(A=@n81, B=NamedRerouteUsage(), Alpha=NamedRerouteUsage()), Alpha=ScalarParameter() {parameterName=Anim_DMGGlow})
  n83 = Saturate(None=Add(A=@n82, B=@n70.SourceOutput))
  NamedRerouteDeclaration(None=MaterialFunctionCall(Default=LinearInterpolate(A=@n82, B=@n83) {constAlpha=0.3333}, LowEnd=@n82) {materialFunction=&MF_UI_LowEnd_ValueSwitcher}.Result) {name=FillGradientColor, nodeColor=<rgba 0 0 0 1>}
  n84 = ComponentMask(None=NamedRerouteDeclaration(None=@n73.UVs) {name=BarUVs, nodeColor=<rgba 0 0 0 1>}) {r=true}
  n85 = LinearInterpolate(
   A=NamedRerouteDeclaration(None=ScalarParameter() {defaultValue=1, group=Fill, parameterName=Progress, sortPriority=0}) {name=Progress, nodeColor=<rgba 0 0 0 1>}
   Alpha=NamedRerouteUsage()
  )
  n86 = Ceil(None=NamedRerouteDeclaration(None=Subtract(A=@n84, B=@n85) {constB=0.75}) {name=ProgressFillSDF, nodeColor=<rgba 0 0 0 1>})
  NamedRerouteDeclaration(None=OneMinus(None=Saturate(None=@n86))) {name=ProgressFillMask, nodeColor=<rgba 0 0 0 1>}
  NamedRerouteDeclaration(None=@n35) {name=BarTopHighlightMask, nodeColor=<rgba 0 0 0 1>}
  n87 = LinearInterpolate(B=ScalarParameter() {defaultValue=0.3, group=Color, parameterName=FillHighlightOpacity, sortPriority=3}, Alpha=NamedRerouteUsage())
  n88 = ConstantBiasScale(None=OneMinus(None=@n37)) {bias=-0.5, scale=2}
  n89 = Saturate(None=@n88)
  n90 = ScalarParameter() {defaultValue=0.2, group=Color, parameterName=FillShadowOpacity, sortPriority=3}
  n91 = LinearInterpolate(A=LinearInterpolate(A=OneMinus(None=@n90), Alpha=NamedRerouteUsage()), Alpha=Saturate(None=Multiply(A=@n88) {constB=-1}))
  n92 = Saturate(None=MaterialFunctionCall!(Blend=LinearInterpolate(B=@n87, Alpha=@n89), Base=Multiply(A=@n81, B=@n91)) {materialFunction=/Engine/Functions/Engine_MaterialFunctions03/Blends/Blend_ColorDodge}.Result)
  n93 = OneMinus(None=NamedRerouteUsage())
  n94 = LinearInterpolate(
   A=VectorParameter() {group=Color, parameterName=MarkerColor, sortPriority=4, defaultValue=<rgba 0.9301 1 0.0103 1>}.RGB
   B=VectorParameter() {group=DBNO, parameterName=DBNOMarkerColor, sortPriority=3, defaultValue=<rgba 1 0.1253 0.0991 1>}.RGB
   Alpha=@n80
  )
  n95 = ScalarParameter() {parameterName=MarkerIntensity}
  NamedRerouteDeclaration(None=LinearInterpolate(A=Multiply(A=@n92, B=@n93), B=LinearInterpolate(A=@n94, Alpha=@n95), Alpha=NamedRerouteUsage())) {name=FillColor, nodeColor=<rgba 0 0 0 1>}
 ]
}
