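# UI material that draws triangle particles for the damage delta on the player vitals HUD.
# Three panning, perspective-warped particle grids and a hand-placed triangle render procedural shapes sized by eased damage amount and animation parameters, with a static switch for fewer triangles.
# Parameters tint it red, shield cyan or vehicle color.
# Alpha-composite output allows additive mixing, and a viewport-scaled position offset moves the effect.
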Material /BattleRoyaleGameUI/HUD/PlayerHitPointInfo/Assets/M_UI_PlayerVitals_DMGDelta_Particles
graph {
 properties {bAutomaticallySetUsageInEditor=false, blendMode=BLEND_AlphaComposite, materialDomain=MD_UI, usedShadingModels=Default Lit}
 nodes [
  n1 = LinearInterpolate(
   A=ScalarParameter() {defaultValue=-344, parameterName=X_Offset_Min}
   B=ScalarParameter() {parameterName=X_Offset_Max}
   Alpha=ScalarParameter() {defaultValue=1, parameterName=Position}
  )
  n2 = ScalarParameter() {parameterName=Y_Offset}
  Add(
   A=WorldPosition().XY
   B=Multiply(
    A=AppendVector(A=@n1, B=@n2)
    B=MaterialFunctionCall() {materialFunction=/FortUILibrary/MaterialFunctions/Basics/MF_UI_ViewportScale}.Viewport Scale Factor
   )
  )
  n3 = Multiply(A=NamedRerouteUsage(), B=OneMinus(None=NamedRerouteUsage()))
  n4 = LinearInterpolate(Alpha=@n3) {constB=2}
  LinearInterpolate(A=Multiply(A=@n4) {constB=-1}, B=@n4, Alpha=@n3) {constA=-1}
  n5 = Multiply(A=NamedRerouteUsage(), B=OneMinus(None=NamedRerouteUsage()))
  n6 = LinearInterpolate(Alpha=@n5) {constB=4}
  LinearInterpolate(A=Multiply(A=@n6) {constB=-1}, B=@n6, Alpha=@n5) {constA=-1}
  n7 = MaterialFunctionCall(
   UVs=MaterialFunctionCall() {materialFunction=&MF_UI_UVScaling}.ScaleY
   "Rotation Angle (0-1)"=ScalarParameter() {parameterName=UV_Rotation}
  ) {materialFunction=&CustomRotator}
  MaterialFunctionCall!(UV=@n7.Rotated Values, Mirror Y=Constant(), Swizzle UV=Constant() {r=1}) {materialFunction=/FortUILibrary/MaterialFunctions/Basics/MF_UI_Symmetry_2x1_New}
  n8 = MaterialFunctionCall() {materialFunction=&GetUserInterfaceUV}
  n9 = MaterialFunctionCall(Float2=@n8.Normalized UV) {materialFunction=&BreakOutFloat2Components}
  MaterialFunctionCall(UV=@n9.G, Mirror Y=Constant(), Swizzle UV=Constant() {r=1}) {materialFunction=&MF_UI_Symmetry_2x1_New}
  n10 = NamedRerouteUsage()
  n11 = NamedRerouteUsage()
  n12 = LinearInterpolate(A=NamedRerouteUsage(), Alpha=LinearInterpolate(A=@n11, Alpha=NamedRerouteUsage()))
  n13 = Saturate(None=StaticSwitchParameter(True=LinearInterpolate(A=@n10, Alpha=@n11), False=LinearInterpolate(A=@n10, Alpha=@n12)) {parameterName=MinimalTriangles})
  Multiply(A=@n13, B=OneMinus(None=ScalarParameter() {parameterName=AdditiveMixRatio}))
  n14 = CollectionParameter() {collection=/BattleRoyaleGameUI/Shared/MPC_BRGameUI, parameterName=Red}
  n15 = CollectionParameter() {collection=&MPC_BRGameUI, parameterName=Cyan}
  n16 = LinearInterpolate(
   A=LinearInterpolate(
    A=ComponentMask(None=@n14) {b=true, g=true, r=true}
    B=ComponentMask(None=@n15) {b=true, g=true, r=true}
    Alpha=ScalarParameter() {parameterName=IsShieldDMG}
   )
   B=VectorParameter() {parameterName=VehicleColor, defaultValue=<rgba 1 0.85 0.1 1>}.RGB
   Alpha=ScalarParameter() {parameterName=IsVehicleDMG}
  )
  Multiply(A=VertexColor().A, B=Multiply(A=@n16, B=@n13))
  n17 = MaterialFunctionCall!(
   UV=NamedRerouteUsage()
   Perspective Rotation=ScalarParameter() {parameterName=ParticleSet2_PerspectiveRotation}
   Perspective Amount=ScalarParameter() {parameterName=ParticleSet2_PerspectiveAmount}
  ) {materialFunction=/FortUILibrary/MaterialFunctions/Basics/MF_UI_Perspective}
  n18 = NamedRerouteUsage()
  n19 = ScalarParameter() {defaultValue=0.25, parameterName=Set2_TimeOffset}
  n20 = AppendVector(A=ScalarParameter() {parameterName=ParticleSet2_PanSpeed_X}, B=ScalarParameter() {defaultValue=-0.5, parameterName=ParticleSet2_PanSpeed_Y})
  n21 = NamedRerouteDeclaration(None=Panner(Coordinate=@n17.Perspective UVs, Time=Add(A=@n18, B=@n19), Speed=Multiply(A=@n20) {constB=-1})) {name=ParticleSet2_Uvs, nodeColor=<rgba 0 0 0 1>}
  n22 = ScalarParameter() {defaultValue=1, parameterName=ParticleSet2_Tiling}
  n23 = Constant()
  n24 = Multiply(
   A=LinearInterpolate(
    A=ScalarParameter() {defaultValue=0.5, parameterName=Set2_MinCellScale}
    B=ScalarParameter() {defaultValue=1, parameterName=Set2_MaxCellSize}
    Alpha=NamedRerouteUsage()
   ) {constB=0.5}
   B=NamedRerouteUsage()
  )
  n25 = NamedRerouteUsage()
  n26 = LinearInterpolate(
   A=LinearInterpolate(A=Multiply(A=@n24, B=@n25), B=NamedRerouteUsage(), Alpha=ScalarParameter() {parameterName=Anim_StartMask})
   Alpha=ScalarParameter() {parameterName=Anim_ScaleDown}
  ) {constB=0}
  n27 = Multiply(A=Multiply(A=NamedRerouteUsage()) {constB=0.5}, B=NamedRerouteUsage())
  NamedRerouteDeclaration(None=MaterialFunctionCall!(
    UVs=MaterialFunctionCall!(UV=@n21, Global Offset, Global Rotation, Tiling=@n22, Cell Offset=@n23, Cell Scale=@n26, Cell Rotation=@n27, Random Seed) {materialFunction=/FortUILibrary/MaterialFunctions/Basics/MF_UI_ParticleGrid}.Cells UV
    Corner 01
    Corner 02
    Corner 03
    Glow Min
    Glow Max
    Outline Thickness
   ) {materialFunction=/FortUILibrary/MaterialFunctions/Shapes/MF_UI_WIP_Triangle}.Fill) {name=Set2, nodeColor=<rgba 0 1 0.4844 1>}
  n28 = MaterialFunctionCall("Value (0-1)"=ScalarParameter() {defaultValue=0.25, parameterName=DMGAmountSizeController}) {materialFunction=&MF_UI_EaseCurves}
  NamedRerouteDeclaration(None=LinearInterpolate(Alpha=Saturate(None=@n28.EaseOutSine)) {constA=0.25}) {name=DMGSizeMultiplier, nodeColor=<rgba 0 0 0 1>}
  n29 = MaterialFunctionCall(UV=NamedRerouteUsage(), Tiling=ScalarParameter() {defaultValue=2, parameterName=ParticleSet1_Tiling}, Random Seed=Constant() {r=25.5603}) {materialFunction=&MF_UI_ParticleGrid}
  NamedRerouteDeclaration(None=@n29.Random Cell Values) {name=Set1_RandomCells, nodeColor=<rgba 0 0 0 1>}
  NamedRerouteDeclaration(None=Multiply(A=Time(), B=ScalarParameter() {defaultValue=1, parameterName=TimeScalar})) {name=TimeControl, nodeColor=<rgba 0 0 0 1>}
  NamedRerouteDeclaration(None=@n17.Perspective UVs) {name=ParticleSet2_PerspectiveUVs, nodeColor=<rgba 0 0 0 1>}
  NamedRerouteDeclaration(None=MaterialFunctionCall(UV=NamedRerouteUsage(), Tiling=ScalarParameter() {defaultValue=1, parameterName=ParticleSet2_Tiling}, Random Seed=Constant() {r=7.0665}) {materialFunction=&MF_UI_ParticleGrid}.Random Cell Values) {name=Set2_RandomCells, nodeColor=<rgba 0 0 0 1>}
  n30 = MaterialFunctionCall(
   UV=NamedRerouteUsage()
   Perspective Rotation=ScalarParameter() {parameterName=ParticleSet3_PerspectiveRotation}
   Perspective Amount=ScalarParameter() {parameterName=ParticleSet3_PerspectiveAmount}
  ) {materialFunction=&MF_UI_Perspective}
  NamedRerouteDeclaration(None=@n30.Perspective UVs) {name=ParticleSet3_PerspectiveUVs, nodeColor=<rgba 0 0 0 1>}
  NamedRerouteDeclaration(None=MaterialFunctionCall(UV=NamedRerouteUsage(), Tiling=ScalarParameter() {defaultValue=1, parameterName=ParticleSet3_Tiling}, Random Seed=Constant() {r=20.4487}) {materialFunction=&MF_UI_ParticleGrid}.Random Cell Values) {name=Set3_RandomCells, nodeColor=<rgba 0 0 0 1>}
  n31 = NamedRerouteUsage()
  n32 = ScalarParameter() {defaultValue=0.125, parameterName=Set3_TimeOffset}
  n33 = AppendVector(A=ScalarParameter() {parameterName=ParticleSet3_PanSpeed_X}, B=ScalarParameter() {defaultValue=-0.25, parameterName=ParticleSet3_PanSpeed_Y})
  n34 = NamedRerouteDeclaration(None=Panner(Coordinate=@n30.Perspective UVs, Time=Add(A=@n31, B=@n32), Speed=Multiply(A=@n33) {constB=-1})) {name=ParticleSet3_Uvs, nodeColor=<rgba 0 0 0 1>}
  n35 = ScalarParameter() {defaultValue=1, parameterName=ParticleSet3_Tiling}
  n36 = Constant()
  n37 = Multiply(
   A=LinearInterpolate(
    A=ScalarParameter() {defaultValue=0.5, parameterName=Set3_MinCellScale}
    B=ScalarParameter() {defaultValue=1, parameterName=Set3_MaxCellSize}
    Alpha=NamedRerouteUsage()
   ) {constB=0.5}
   B=NamedRerouteUsage()
  )
  n38 = NamedRerouteUsage()
  n39 = LinearInterpolate(
   A=LinearInterpolate(A=Multiply(A=@n37, B=@n38), B=NamedRerouteUsage(), Alpha=ScalarParameter() {parameterName=Anim_StartMask})
   Alpha=ScalarParameter() {parameterName=Anim_ScaleDown}
  ) {constB=0}
  n40 = Multiply(A=Multiply(A=NamedRerouteUsage()) {constB=0.5}, B=NamedRerouteUsage())
  NamedRerouteDeclaration(None=MaterialFunctionCall(UVs=MaterialFunctionCall(UV=@n34, Tiling=@n35, Cell Offset=@n36, Cell Scale=@n39, Cell Rotation=@n40) {materialFunction=&MF_UI_ParticleGrid}.Cells UV) {materialFunction=&MF_UI_WIP_Triangle}.Fill) {name=Set3, nodeColor=<rgba 1 0.6328 0 1>}
  NamedRerouteDeclaration(None=MaterialFunctionCall(Float2=@n8.Pixel Size) {materialFunction=&BreakOutFloat2Components}.R) {name=X_PixelSize, nodeColor=<rgba 0 0 0 1>}
  n41 = Constant()
  n42 = LinearInterpolate(Alpha=ScalarParameter() {parameterName=Anim_ManualPan}) {constA=-1, constB=0}
  n43 = MaterialFunctionCall(
   UVs=Add(A=MaterialFunctionCall() {materialFunction=&MF_UI_UVScaling}.ScaleX, B=AppendVector(A=@n41, B=@n42))
   "Rotation Angle (0-1)"=Constant() {r=0.05}
  ) {materialFunction=&CustomRotator}
  n44 = LinearInterpolate(
   A=Multiply(A=ScalarParameter() {defaultValue=0.75, parameterName=ManualTriangle_1_Size}, B=NamedRerouteUsage())
   Alpha=ScalarParameter() {parameterName=Anim_ScaleDown}
  ) {constA=0.75, constB=0}
  NamedRerouteDeclaration(None=MaterialFunctionCall(UVs=MaterialFunctionCall(UVs=@n43.Rotated Values, Texture Scale=@n44) {materialFunction=&ScaleUVsByCenter}.UVs) {materialFunction=&MF_UI_WIP_Triangle}.Fill) {name=ManualTriangle_1, nodeColor=<rgba 0 1 0.5313 1>}
  NamedRerouteDeclaration(None=@n29.Random Columns) {name=Set1_RandomColumns, nodeColor=<rgba 0 0 0 1>}
  NamedRerouteDeclaration(None=@n7.Rotated Values) {name=ScaledUVs, nodeColor=<rgba 0 0 0 1>}
  NamedRerouteDeclaration(None=Saturate(None=@n9.G)) {name=CellScaleRamp, nodeColor=<rgba 0 0 0 1>}
  NamedRerouteDeclaration(None=SmoothStep(Value=NamedRerouteUsage()) {constMax=2, constMin=0.25}) {name=AnimStartMask, nodeColor=<rgba 0 0 0 1>}
  n45 = MaterialFunctionCall(
   UV=NamedRerouteUsage()
   Perspective Rotation=ScalarParameter() {parameterName=ParticleSet1_PerspectiveRotation}
   Perspective Amount=ScalarParameter() {parameterName=ParticleSet1_PerspectiveAmount}
  ) {materialFunction=&MF_UI_Perspective}
  n46 = ScalarParameter() {parameterName=ParticleSet1_PanSpeed_X}
  n47 = MaterialFunctionCall(Tiling=ScalarParameter() {defaultValue=2, parameterName=ParticleSet1_Tiling}) {materialFunction=&MF_UI_ParticleGrid}
  n48 = Multiply(
   A=ScalarParameter() {defaultValue=0.25, parameterName=ParticleSet1_PanSpeed_Y}
   B=Power(Base=SmoothStep(Value=@n47.Random Columns)) {constExponent=0.5}
  )
  n49 = NamedRerouteDeclaration(None=Panner(
    Coordinate=NamedRerouteDeclaration(None=@n45.Perspective UVs) {name=ParticleSet1_PerspectiveUVs, nodeColor=<rgba 0 0 0 1>}
    Time=NamedRerouteUsage()
    Speed=AppendVector(A=@n46, B=@n48)
   )) {name=ParticleSet1_Uvs, nodeColor=<rgba 0 0 0 1>}
  n50 = ScalarParameter() {defaultValue=2, parameterName=ParticleSet1_Tiling}
  n51 = Constant()
  n52 = Multiply(
   A=LinearInterpolate(
    A=ScalarParameter() {defaultValue=0.5, parameterName=Set1_MinCellScale}
    B=ScalarParameter() {defaultValue=1, parameterName=Set1_MaxCellSize}
    Alpha=NamedRerouteUsage()
   ) {constB=0.5}
   B=NamedRerouteUsage()
  )
  n53 = NamedRerouteUsage()
  n54 = LinearInterpolate(
   A=LinearInterpolate(A=Multiply(A=@n52, B=@n53), B=NamedRerouteUsage(), Alpha=ScalarParameter() {parameterName=Anim_StartMask})
   Alpha=ScalarParameter() {parameterName=Anim_ScaleDown}
  ) {constB=0}
  n55 = Multiply(A=Multiply(A=NamedRerouteUsage()) {constB=0.5}, B=NamedRerouteUsage())
  NamedRerouteDeclaration(None=MaterialFunctionCall(UVs=MaterialFunctionCall(UV=@n49, Tiling=@n50, Cell Offset=@n51, Cell Scale=@n54, Cell Rotation=@n55) {materialFunction=&MF_UI_ParticleGrid}.Cells UV) {materialFunction=&MF_UI_WIP_Triangle}.Fill) {name=Set1, nodeColor=<rgba 0.8047 0 1 1>}
  Saturate(None=@n16)
 ]
}
