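# Fortnite Battle Royale HUD widget showing one player vital bar (health, shield or vehicle health) with icon, numeric value and damage effects.
# BarType picks the materials, with the bar material async loaded.
# Value changes push progress into dynamic bar, icon and font materials, and damage plays impact and shake animations that intensify when low.
# It also handles downed (DBNO) visuals, optional hiding of shields and a tag-gated max-value separator.
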WidgetBlueprint /BattleRoyaleGameUI/HUD/PlayerHitPointInfo/WBP_BRGameUI_PlayerVitalBar
animations [
 Anim_BarShake {
  binding_count=1
  length_seconds=0.1849
  object=&self:Anim_BarShake
  sequence {bindings [Overlay_Bar {
     id=#1
     possessed_class=Overlay
     tracks [{
       type=MovieScene2DTransformTrack
       sections [{
         type=MovieScene2DTransformSection
         end_frame=3
         channels [
          Translation.X {type=MovieSceneScriptingFloatChannel, keys [[0,0],[0.8513,-6.8849],[1.4223,3.9219],[2.184,-3.421],[2.7867,0.8919],[3.698,0]]}
          Translation.Y {type=MovieSceneScriptingFloatChannel, keys [[0,0],[0.8513,6.9314],[1.4223,-5.0319],[2.184,4.061],[2.7867,-2.5554],[3.698,0]]}
          Angle {type=MovieSceneScriptingFloatChannel}
          Scale.X {type=MovieSceneScriptingFloatChannel}
          Scale.Y {type=MovieSceneScriptingFloatChannel}
          Shear.X {type=MovieSceneScriptingFloatChannel}
          Shear.Y {type=MovieSceneScriptingFloatChannel}
         ]
        }]
      }]
    }]}
 }
 Anim_BarShake_1 {
  binding_count=1
  length_seconds=0.1849
  object=&self:Anim_BarShake_1
  sequence {bindings [Overlay_Bar {
     id=#2
     possessed_class=Overlay
     tracks [{
       type=MovieScene2DTransformTrack
       sections [{
         type=MovieScene2DTransformSection
         end_frame=3
         channels [
          Translation.X {type=MovieSceneScriptingFloatChannel, keys [[0,0],[0.8513,-6.8849],[1.4223,3.9219],[2.184,-3.421],[2.7867,0.8919],[3.698,0]]}
          Translation.Y {type=MovieSceneScriptingFloatChannel, keys [[0,0],[0.8513,6.9314],[1.4223,-5.0319],[2.184,4.061],[2.7867,-2.5554],[3.698,0]]}
          Angle {type=MovieSceneScriptingFloatChannel}
          Scale.X {type=MovieSceneScriptingFloatChannel}
          Scale.Y {type=MovieSceneScriptingFloatChannel}
          Shear.X {type=MovieSceneScriptingFloatChannel}
          Shear.Y {type=MovieSceneScriptingFloatChannel}
         ]
        }]
      }]
    }]}
 }
 Anim_BarShake_2 {
  binding_count=1
  length_seconds=0.1849
  object=&self:Anim_BarShake_2
  sequence {bindings [Overlay_Bar {
     id=#3
     possessed_class=Overlay
     tracks [{
       type=MovieScene2DTransformTrack
       sections [{
         type=MovieScene2DTransformSection
         end_frame=3
         channels [
          Translation.X {type=MovieSceneScriptingFloatChannel, keys [[0,0],[0.8513,-9.27],[1.4223,6.3372],[2.184,-5.474],[2.5827,1.7071],[3.698,0]]}
          Translation.Y {type=MovieSceneScriptingFloatChannel, keys [[0,0],[0.8513,9.075],[1.4223,-7.2057],[2.184,4.061],[2.5953,-2.9479],[3.698,0]]}
          Angle {type=MovieSceneScriptingFloatChannel}
          Scale.X {type=MovieSceneScriptingFloatChannel}
          Scale.Y {type=MovieSceneScriptingFloatChannel}
          Shear.X {type=MovieSceneScriptingFloatChannel}
          Shear.Y {type=MovieSceneScriptingFloatChannel}
         ]
        }]
      }]
    }]}
 }
 Anim_ValueImpact {
  binding_count=3
  length_seconds=1
  object=&self:Anim_ValueImpact
  sequence {
   bindings [
    CurrentValue {
     id=#4
     possessed_class=CommonTextBlock
     tracks [
      {
       type=MovieScene2DTransformTrack
       sections [{
         type=MovieScene2DTransformSection
         channels [
          Scale.X {type=MovieSceneScriptingFloatChannel, keys [[0,1],[3,1.75],[24,1]]}
          Scale.Y {type=MovieSceneScriptingFloatChannel, keys [[0,1],[3,1.75],[24,1]]}
         ]
        }]
      }
      {
       type=MovieSceneWidgetMaterialTrack
       display_name=Font.FontMaterial
       sections [{type=MovieSceneParameterSection, channels [Anim_Damage {type=MovieSceneScriptingFloatChannel, keys [[0,0],[3,1],[6,1],[24,0]]}]}]
      }
     ]
    }
    Image_ValueIcon {
     id=#5
     possessed_class=Image
     tracks [
      {
       type=MovieSceneWidgetMaterialTrack
       sections [{type=MovieSceneParameterSection, channels [Anim_Color {type=MovieSceneScriptingFloatChannel, keys [[0,0],[3,1],[30,0]]}]}]
      }
      {
       type=MovieScene2DTransformTrack
       sections [{
         type=MovieScene2DTransformSection
         channels [
          Translation.X {type=MovieSceneScriptingFloatChannel, const [0,1]}
          Translation.Y {type=MovieSceneScriptingFloatChannel, const [0,1]}
          Angle {type=MovieSceneScriptingFloatChannel, keys [[0,0],[3,-6],[30,0]]}
          Scale.X {type=MovieSceneScriptingFloatChannel, keys [[0,1],[3,1.75],[30,1]]}
          Scale.Y {type=MovieSceneScriptingFloatChannel, keys [[0,1],[3,1.75],[30,1]]}
          Shear.X {type=MovieSceneScriptingFloatChannel, const [0,1]}
          Shear.Y {type=MovieSceneScriptingFloatChannel, const [0,1]}
         ]
        }]
      }
     ]
    }
    Image_Bar {
     id=#6
     possessed_class=Image
     tracks [{
       type=MovieSceneWidgetMaterialTrack
       sections [{
         type=MovieSceneParameterSection
         channels [
          MarkerIntensity {type=MovieSceneScriptingFloatChannel, keys [[0,0],[1,1],[6,0]]}
          MarkerSmoothing {type=MovieSceneScriptingFloatChannel, keys [[0,0],[2,0.5],[10,0]]}
          Anim_DMGGlow {type=MovieSceneScriptingFloatChannel, keys [[0,0],[2,1],[15,1],[60,0]]}
         ]
        }]
      }]
    }
   ]
   display_rate {numerator=60}
  }
 }
 Anim_ValueImpact_WithTriangleParticles {
  binding_count=7
  length_seconds=1
  object=&self:Anim_ValueImpact_WithTriangleParticles
  sequence {
   bindings [
    DeltaParticles_Top {
     id=#7
     possessed_class=Image
     tracks [
      {
       type=MovieSceneWidgetMaterialTrack
       sections [{
         type=MovieSceneParameterSection
         channels [
          Anim_ScaleDown {type=MovieSceneScriptingFloatChannel, keys [[0,0],[60,1]]}
          Anim_StartMask {type=MovieSceneScriptingFloatChannel, keys [[0,1],[2,0]]}
          Anim_ManualPan {type=MovieSceneScriptingFloatChannel, keys [[0,0],[60,1]]}
         ]
        }]
      }
      {
       type=MovieSceneByteTrack
       display_name=Visibility
       sections [{type=MovieSceneByteSection, channels [None {type=MovieSceneScriptingByteChannel, keys [[0,4],[60,2]]}]}]
      }
     ]
    }
    DeltaParticles_Bottom {
     id=#8
     possessed_class=Image
     tracks [
      {
       type=MovieSceneWidgetMaterialTrack
       sections [{
         type=MovieSceneParameterSection
         channels [
          Anim_ScaleDown {type=MovieSceneScriptingFloatChannel, keys [[0,0],[60,1]]}
          Anim_StartMask {type=MovieSceneScriptingFloatChannel, keys [[0,1],[2,0]]}
          Anim_ManualPan {type=MovieSceneScriptingFloatChannel, keys [[0,0],[60,1]]}
         ]
        }]
      }
      {
       type=MovieSceneByteTrack
       display_name=Visibility
       sections [{type=MovieSceneByteSection, channels [None {type=MovieSceneScriptingByteChannel, keys [[0,4],[60,2]]}]}]
      }
     ]
    }
    CurrentValue {
     id=#4
     possessed_class=CommonTextBlock
     tracks [
      {
       type=MovieScene2DTransformTrack
       sections [{
         type=MovieScene2DTransformSection
         channels [
          Scale.X {type=MovieSceneScriptingFloatChannel, keys [[0,1],[3,1.75],[24,1]]}
          Scale.Y {type=MovieSceneScriptingFloatChannel, keys [[0,1],[3,1.75],[24,1]]}
         ]
        }]
      }
      {
       type=MovieSceneWidgetMaterialTrack
       display_name=Font.FontMaterial
       sections [{type=MovieSceneParameterSection, channels [Anim_Damage {type=MovieSceneScriptingFloatChannel, keys [[0,0],[3,1],[6,1],[24,0]]}]}]
      }
     ]
    }
    DeltaParticles_Horizontal {
     id=#9
     possessed_class=Image
     tracks [
      {
       type=MovieSceneWidgetMaterialTrack
       sections [{
         type=MovieSceneParameterSection
         channels [
          Anim_ScaleDown {type=MovieSceneScriptingFloatChannel, keys [[0,0],[60,1]]}
          Anim_StartMask {type=MovieSceneScriptingFloatChannel, keys [[0,1],[2,0]]}
          Anim_ManualPan {type=MovieSceneScriptingFloatChannel, keys [[0,0],[60,1]]}
         ]
        }]
      }
      {
       type=MovieSceneByteTrack
       display_name=Visibility
       sections [{type=MovieSceneByteSection, channels [None {type=MovieSceneScriptingByteChannel, keys [[0,4],[60,2]]}]}]
      }
     ]
    }
    Image_ValueIcon {
     id=#5
     possessed_class=Image
     tracks [
      {
       type=MovieSceneWidgetMaterialTrack
       sections [{type=MovieSceneParameterSection, channels [Anim_Color {type=MovieSceneScriptingFloatChannel, keys [[0,0],[3,1],[30,0]]}]}]
      }
      {
       type=MovieScene2DTransformTrack
       sections [{
         type=MovieScene2DTransformSection
         channels [
          Translation.X {type=MovieSceneScriptingFloatChannel, const [0,1]}
          Translation.Y {type=MovieSceneScriptingFloatChannel, const [0,1]}
          Angle {type=MovieSceneScriptingFloatChannel, keys [[0,0],[3,-6],[30,0]]}
          Scale.X {type=MovieSceneScriptingFloatChannel, keys [[0,1],[3,1.75],[30,1]]}
          Scale.Y {type=MovieSceneScriptingFloatChannel, keys [[0,1],[3,1.75],[30,1]]}
          Shear.X {type=MovieSceneScriptingFloatChannel, const [0,1]}
          Shear.Y {type=MovieSceneScriptingFloatChannel, const [0,1]}
         ]
        }]
      }
     ]
    }
    Image_Bar {
     id=#6
     possessed_class=Image
     tracks [{
       type=MovieSceneWidgetMaterialTrack
       sections [{
         type=MovieSceneParameterSection
         channels [
          MarkerIntensity {type=MovieSceneScriptingFloatChannel, keys [[0,0],[1,1],[6,0]]}
          MarkerSmoothing {type=MovieSceneScriptingFloatChannel, keys [[0,0],[2,0.5],[10,0]]}
          Anim_DMGGlow {type=MovieSceneScriptingFloatChannel, keys [[0,0],[2,1],[15,1],[60,0]]}
         ]
        }]
      }]
    }
    ParticlesContainer {
     id=#10
     possessed_class=Overlay
     tracks [
      {
       type=MovieScene2DTransformTrack
       sections [{
         type=MovieScene2DTransformSection
         channels [
          Scale.X {type=MovieSceneScriptingFloatChannel, keys [[0,0],[60,2]]}
          Scale.Y {type=MovieSceneScriptingFloatChannel, keys [[0,0],[60,2]]}
         ]
        }]
      }
      {
       type=MovieSceneByteTrack
       display_name=Visibility
       sections [{type=MovieSceneByteSection, channels [None {type=MovieSceneScriptingByteChannel, keys [[0,4],[60,2]]}]}]
      }
     ]
    }
   ]
   display_rate {numerator=60}
  }
 }
]
graphs [
 EventGraph {nodes [
   OnBarShakeAnimationFinished() {is=title K2Node_CustomEvent}
   > InvalidateLayoutAndVolatility(self:Widget Object Reference)
   InitializeBar() {is=title K2Node_CustomEvent}
   > n1 = Load Soft Hit Bar Material!(
    execute:Exec
    Target:Image Object Reference=Get Image_Bar() -> (Image_Bar:Image Object Reference)
    SoftMaterial:Material Interface Soft Object Reference=Select!(
     Health:Material Interface Soft Object Reference=/BattleRoyaleGameUI/HUD/PlayerHitPointInfo/Assets/M_UI_HealthBar_Master
     Shield:Material Interface Soft Object Reference=/BattleRoyaleGameUI/HUD/PlayerHitPointInfo/Assets/MI_UI_ShieldBar
     Overshield:Material Interface Soft Object Reference
     Armor:Material Interface Soft Object Reference
     Stamina:Material Interface Soft Object Reference
     VehicleHealth:Material Interface Soft Object Reference=/BattleRoyaleGameUI/HUD/PlayerHitPointInfo/Assets/MI_UI_VehicleHealthBar
     SignalInStorm:Material Interface Soft Object Reference
     Index:EHealthBarType Enum=Get BarType() -> (BarType:EHealthBarType Enum=Health)
    ) -> (ReturnValue:Material Interface Soft Object Reference)
   ) -> (then:Exec, LoadCompleted:Exec) {is=title K2Node_MacroInstance}
   > RefreshIcons()
   > SetPadding!(
    execute
    self:Overlay Slot Object Reference=SlotAsOverlaySlot(Widget=Get HBox_TextArea() -> (HBox_TextArea:Horizontal Box Object Reference))
    InPadding:Margin Structure=Make Margin(Left=16, Top=0, Right=0, Bottom=0)
   ) -> (then)
   n2 = Get SizeBox_Bar() -> (SizeBox_Bar:Size Box Object Reference)
   > SetHeightOverride(self=@n2, InHeightOverride=Get BarHeight() -> (BarHeight:double))
   > SetWidthOverride(self=@n2, InWidthOverride=Get BarWidth() -> (BarWidth:double))
   > SetupFXVisuals()
   Branch(execute=@n1.LoadCompleted, Condition=Get HasReceivedValueUpdate() -> (HasReceivedValueUpdate:bool))
   > UpdateCurrentValue()
   OnValueImpactAnimationFinished() {is=title K2Node_CustomEvent}
   > SetRenderScale!(execute, self:Widget Object Reference=Get CurrentValue() -> (CurrentValue:Common Text Object Reference), Scale:Vector 2D Structure=(X=1,Y=1)) -> (then)
   n3 = OnViewTargetChanged!() -> (
    OutputDelegate
    then
    PlayerController:Fort Player Controller Object Reference
    OldViewTarget:Actor Object Reference
    NewViewTarget:Actor Object Reference
   ) {is=title K2Node_CustomEvent}
   > Set ViewTarget(ViewTarget:Actor Object Reference=@n3.NewViewTarget) -> (Output_Get:Actor Object Reference)
   > Set IsViewTargetChanged(IsViewTargetChanged:bool=true) -> (Output_Get:bool)
   Event OnDeltaChanged() {is=title K2Node_Event}
   > Update Delta Bar()
   Event Construct()
   > InitializeBar() {is=title K2Node_CallFunction}
   > BindToAnimationFinished(Animation=Get Anim_BarShake() -> (Anim_BarShake:Widget Animation Object Reference), Delegate=Create Event())
   > BindToAnimationFinished(Animation=Get Anim_ValueImpact() -> (Anim_ValueImpact:Widget Animation Object Reference), Delegate=Create Event())
   > n4 = Cast To FortPlayerController(Object=GetOwningPlayer())
   > n5 = Set PlayerController(PlayerController:Fort Player Controller Object Reference=@n4) -> (Output_Get:Fort Player Controller Object Reference)
   > Bind Event to On View Target Changed(self:Fort Player Controller Object Reference=@n5, Delegate=Create Event()) {is=title K2Node_AddDelegate}
   n6 = Event PreConstruct()
   > Set HUDElementTag!(
    execute
    HUDElementTag:Gameplay Tag Container Structure=Get TextHudElementTags() -> (TextHudElementTags:Gameplay Tag Container Structure)
    self:Fort HUDElement Widget Object Reference=Get FortHUDElementWrapper_Text() -> (FortHUDElementWrapper_Text:Fort HUDElement Wrapper Object Reference)
   ) -> (then, Output_Get:Gameplay Tag Container Structure)
   > Branch(Condition=@n6.IsDesignTime)
   > InitializeBar()
   > SetInformationVisibility!(
    execute
    self
    Icon_Visibility:bool=Get bIcon_isVisible() -> (bIcon_isVisible:bool)
    Value_Visibility:bool=Get bValue_isVisible() -> (bValue_isVisible:bool)
   ) -> (then)
   n7 = Event OnMaxValueChanged!() -> (OutputDelegate, then, Value:float) {is=title K2Node_Event}
   > Set HasReceivedDeltaUpdate(HasReceivedDeltaUpdate:bool=true) -> (Output_Get:bool)
   > UpdateMaxValue()
   > HandleAddSeparator!(execute, self, NewMax:double=@n7.Value) -> (then)
   n8 = Event OnDBNOStateChanged!() -> (OutputDelegate, then, IsDBNO:bool) {is=title K2Node_Event}
   > UpdateDBNOState!(execute, self, bIsDBNO:bool=@n8.IsDBNO) -> (then)
   > InvalidateLayoutAndVolatility()
   n9 = Event OnValueChangedWithReason!() -> (OutputDelegate, then, Value:float, Reason:EFortHitPointModificationReason Enum=Invalid) {is=title K2Node_Event}
   > Set ValueChangedReason(ValueChangedReason:EFortHitPointModificationReason Enum=@n9.Reason) -> (Output_Get:EFortHitPointModificationReason Enum=Invalid)
   > Set HasReceivedValueUpdate(HasReceivedValueUpdate:bool=true) -> (Output_Get:bool)
   > UpdateCurrentValue()
   n10 = Event OnExtensionWidgetAdded!() -> (OutputDelegate, then, ExtensionSlot:EUIExtensionSlot Enum=Invalid, ExtensionWidget:User Widget Object Reference) {is=title K2Node_Event}
   > n11 = AddChildToOverlay!(
    execute
    self:Overlay Object Reference=Get Overlay_ExtensionWidget() -> (Overlay_ExtensionWidget:Overlay Object Reference)
    Content:Widget Object Reference=@n10.ExtensionWidget
   ) -> (then, ReturnValue:Overlay Slot Object Reference)
   > SetHorizontalAlignment(self=@n11, InHorizontalAlignment=HAlign_Fill)
   > SetVerticalAlignment(self=@n11, InVerticalAlignment=VAlign_Fill)
   Event Destruct()
   > UnbindAllFromAnimationFinished!(execute, self:User Widget Object Reference, Animation:Widget Animation Object Reference=Get Anim_BarShake()) -> (then)
   > UnbindAllFromAnimationFinished(Animation=Get Anim_ValueImpact())
   > n12 = Get!(execute:Exec) -> (then:Exec, else:Exec, PlayerController:Fort Player Controller Object Reference)
   > Unbind all Events from On View Target Changed!(execute:Exec, self:Fort Player Controller Object Reference=@n12) -> (then:Exec) {is=title K2Node_ClearDelegate}
  ]}
 HandleAddSeparator {nodes [
   HandleAddSeparator!() -> (then, NewMax:double=~) {is=title K2Node_FunctionEntry}
   > n1 = Branch(Condition=Get IsFirstSetMaxHealth() -> (IsFirstSetMaxHealth:bool))
   > Set IsFirstSetMaxHealth(IsFirstSetMaxHealth:bool=false) -> (Output_Get:bool)
   > Set CachedInitialMaxValue(CachedInitialMaxValue:double=Get NewMax() -> (NewMax:double)) -> (Output_Get:double)
   n2 = HasMatchingGameplayTag!(
    execute:Exec=@n1.else
    self:Object Reference=Select!(
     Option 0:Object Reference=GetOwningPlayer()
     Option 1:Object Reference=Get ViewTarget() -> (ViewTarget:Actor Object Reference)
     Index:bool=Get IsViewTargetChanged() -> (IsViewTargetChanged:bool)
    ) -> (ReturnValue:Object Reference)
    TagToCheck:Gameplay Tag Structure=Get AddSeparatorTag() -> (AddSeparatorTag:Gameplay Tag Structure)
   ) -> (then:Exec, ReturnValue:bool) {is=title K2Node_Message}
   > n3 = GetDynamicMaterial(self=Get Image_Bar())
   n4 = Get bAllowSeparator() -> (bAllowSeparator:bool)
   > SetScalarParameterValue(self=@n3, ParameterName=ShowSeparatorGap, Value=SelectFloat(A=1, B=0, bPickA=AND Boolean(A=@n2, B=@n4)))
   > SetScalarParameterValue(
    self=@n3
    ParameterName=SeparatorGapPosition
    Value=float / float(A=Get CachedInitialMaxValue() -> (CachedInitialMaxValue:double), B=Get NewMax())
   )
  ]}
 LoadSoftHitBarMaterial {nodes [
   n1 = Inputs!() -> (execute, Target:Image Object Reference, SoftMaterial:Material Interface Soft Object Reference)
   > n2 = Async Load Asset(Asset=@n1.SoftMaterial)
   > Outputs!(then:Exec, LoadCompleted:Exec=@n4.then)
   n3 = Cast To MaterialInterface!(execute=@n2.Completed, Object=@n2) -> (then, CastFailed, AsMaterial Interface:Material Interface Object Reference) {is=title K2Node_DynamicCast}
   > n4 = SetBrushFromMaterial(self=@n1, Material=@n3)
  ]}
 RefreshIcons {nodes [
   RefreshIcons() {is=title K2Node_FunctionEntry}
   > SetBrushFromSoftMaterial!(
    execute
    self:Image Object Reference=Get Image_ValueIcon() -> (Image_ValueIcon:Image Object Reference)
    SoftMaterial:Material Interface Soft Object Reference=Select!(
     Health:Material Interface Soft Object Reference=/BattleRoyaleGameUI/HUD/PlayerHitPointInfo/Assets/MI_UI_PlayerVitals_Icon_Health
     Shield:Material Interface Soft Object Reference=/BattleRoyaleGameUI/HUD/PlayerHitPointInfo/Assets/MI_UI_PlayerVitals_Icon_Shields
     Overshield:Material Interface Soft Object Reference
     Armor:Material Interface Soft Object Reference
     Stamina:Material Interface Soft Object Reference
     VehicleHealth:Material Interface Soft Object Reference=/BattleRoyaleGameUI/HUD/PlayerHitPointInfo/Assets/MI_UI_PlayerVitals_Icon_Vehicle
     SignalInStorm:Material Interface Soft Object Reference
     Index:EHealthBarType Enum=Get BarType() -> (BarType=Health)
    ) -> (ReturnValue:Material Interface Soft Object Reference)
   ) -> (then)
  ]}
 SetDMGParticleFXTreatment {nodes [
   n1 = SetDMGParticleFXTreatment!() -> (then, LastValue:double=~, CurrentValue:double=~) {is=title K2Node_FunctionEntry}
   > n2 = For Each Loop(Array:Array of Image Object References=Make Array!(
     "[0]":Image Object Reference=Get DeltaParticles_Bottom() -> (DeltaParticles_Bottom:Image Object Reference)
     "[1]":Image Object Reference=Get DeltaParticles_Top() -> (DeltaParticles_Top:Image Object Reference)
     "[2]":Image Object Reference=Get DeltaParticles_Horizontal() -> (DeltaParticles_Horizontal:Image Object Reference)
    ) -> (Array:Array of Image Object References)) -> (Array Element:Image Object Reference)
   n3 = float - float(A=@n1.CurrentValue)
   n4 = Power!(Base:double=@n3, Exp:double=0.4) -> (ReturnValue:double) {is=title K2Node_CallFunction}
   SetWidthOverride(
    execute=@n2.Completed
    self=Get DMGParticles() -> (DMGParticles:Fort Visual Attachment Object Reference)
    InWidthOverride=Lerp(A=4, B=80, Alpha=@n4)
   )
   > SetHeightOverride(self=Get DMGParticles(), InHeightOverride=Lerp(A=64, B=192, Alpha=@n4))
   n5 = GetDynamicMaterial(execute=@n2.LoopBody, self=@n2)
   > SetScalarParameterValue(self=@n5, ParameterName=Position, Value=Lerp(A=@n1, B=0, Alpha=0.25))
   > SetScalarParameterValue(self=@n5, ParameterName=DMGAmountSizeController, Value=@n3)
  ]}
 SetHealthStyle {nodes [
   SetHealthStyle() {is=title K2Node_FunctionEntry}
   > Branch(Condition="Equal (Enum)"(A:EHealthBarType Enum=Get BarType() -> (BarType=Health), B:EHealthBarType Enum=Health))
   > n1 = GetDynamicMaterial(self=Get Image_Bar())
   > SetScalarParameterValue(self=@n1, ParameterName=IsDBNO, Value="To Float (Boolean)"(InBool=Get DBNOStateCache() -> (DBNOStateCache:bool)))
  ]}
 SetInformationVisibility {nodes [
   SetInformationVisibility!() -> (then, Icon_Visibility:bool=~, Value_Visibility:bool=~) {is=title K2Node_FunctionEntry}
   > n1 = Sequence()
   > SetVisibility(
    self=Get Image_ValueIcon()
    InVisibility=Select!(Option 0:ESlateVisibility Enum=Collapsed, Option 1:ESlateVisibility Enum=HitTestInvisible, Index:bool=Get bIcon_isVisible()) -> (ReturnValue:ESlateVisibility Enum)
   )
   SetVisibility(
    execute=@n1.then_1
    self=Get SizeBox_Value() -> (SizeBox_Value:Size Box Object Reference)
    InVisibility=Select(Option 0=Collapsed, Option 1=SelfHitTestInvisible, Index=Get bValue_isVisible())
   )
  ]}
 SetSize {nodes [
   n1 = SetSize!() -> (then, useLargeSize:bool=~) {is=title K2Node_FunctionEntry}
   n2 = Select(Option 0:double=Get NormalHeight() -> (NormalHeight:double), Option 1:double=Get TallHeight() -> (TallHeight:double), Index=@n1) -> (ReturnValue:double)
   > SetHeightOverride(self=Get SizeBox_Bar(), InHeightOverride=@n2)
   > SetDesiredSizeOverride!(execute, self:Image Object Reference=Get Image_ValueIcon(), DesiredSize:Vector 2D Structure=MakeVector2D(X=@n2, Y=@n2)) -> (then)
  ]}
 SetupFXVisuals {nodes [
   n1 = Get BarType() -> (BarType=Health)
   SetBrushFromMaterial(
    self=Get DeltaParticles_Top()
    Material=Select!(
     Health:Material Interface Object Reference=/BattleRoyaleGameUI/HUD/PlayerHitPointInfo/Assets/M_UI_PlayerVitals_DMGDelta_Particles
     Shield:Material Interface Object Reference=/BattleRoyaleGameUI/HUD/PlayerHitPointInfo/Assets/MI_UI_PlayerVitals_ShieldDelta_Particles
     Overshield:Material Interface Object Reference
     Armor:Material Interface Object Reference
     Stamina:Material Interface Object Reference
     VehicleHealth:Material Interface Object Reference=/BattleRoyaleGameUI/HUD/PlayerHitPointInfo/Assets/MI_UI_PlayerVitals_VehicleDelta_Particles
     SignalInStorm:Material Interface Object Reference
     Index:EHealthBarType Enum=@n1
    ) -> (ReturnValue:Material Interface Object Reference)
   )
   > SetBrushFromMaterial(
    self=Get DeltaParticles_Bottom()
    Material=Select(
     Health=/BattleRoyaleGameUI/HUD/PlayerHitPointInfo/Assets/M_UI_PlayerVitals_DMGDelta_Particles_Bottom
     Shield=/BattleRoyaleGameUI/HUD/PlayerHitPointInfo/Assets/MI_UI_PlayerVitals_ShieldDelta_Particles_Bottom
     VehicleHealth=/BattleRoyaleGameUI/HUD/PlayerHitPointInfo/Assets/MI_UI_PlayerVitals_VehicleDelta_Particles_Bottom
     Index=@n1
    )
   )
   > n2 = SetBrushFromMaterial(
    self=Get DeltaParticles_Horizontal()
    Material=Select(
     Health=/BattleRoyaleGameUI/HUD/PlayerHitPointInfo/Assets/MI_UI_PlayerVitals_DMGDelta_Particles_Horizontal
     Shield=/BattleRoyaleGameUI/HUD/PlayerHitPointInfo/Assets/MI_UI_PlayerVitals_ShieldDelta_Particles_Horizontal
     VehicleHealth=/BattleRoyaleGameUI/HUD/PlayerHitPointInfo/Assets/MI_UI_PlayerVitals_VehicleDelta_Particles_Horizontal
     Index=@n1
    )
   )
   SetFontMaterial!(
    execute=[@n2.then, @n3.then]
    self:Text Object Reference=Get CurrentValue()
    InMaterial:Material Interface Object Reference=Select(
     Health=/BattleRoyaleGameUI/HUD/PlayerHitPointInfo/Assets/M_Font_PlayerVitals_Value
     Shield=/BattleRoyaleGameUI/HUD/PlayerHitPointInfo/Assets/MI_Font_PlayerVitals_Value_Shields
     VehicleHealth=/BattleRoyaleGameUI/HUD/PlayerHitPointInfo/Assets/MI_Font_PlayerVitals_Value_Vehicle
     Index=@n1
    )
   ) -> (then)
   n3 = SetupFXVisuals() {is=title K2Node_FunctionEntry}
   Reroute Node(InputPin:EHealthBarType Enum=@n1) -> (OutputPin:EHealthBarType Enum)
  ]}
 Update Delta Bar {nodes [
   Update Delta Bar() {is=title K2Node_FunctionEntry}
   > n1 = GetDynamicMaterial(self=Get Image_Bar())
   > SetScalarParameterValue(
    self=@n1
    ParameterName=DamageProgress
    Value=GetLastValuePercentage!(self:Athena Player Hit Point Bar Base Object Reference) -> (ReturnValue:float)
   )
  ]}
 Update Fill Bar {nodes [
   Update Fill Bar() {is=title K2Node_FunctionEntry}
   > SetHealthStyle() {is=title K2Node_CallFunction}
   > n1 = GetDynamicMaterial(self=Get Image_Bar())
   n2 = GetCurrentValuePercentage!(self:Athena Player Hit Point Bar Base Object Reference) -> (ReturnValue:float)
   > SetScalarParameterValue(self=@n1, ParameterName=Progress, Value=@n2)
   > n3 = GetDynamicMaterial(self=Get Image_ValueIcon())
   > SetScalarParameterValue(self=@n3, ParameterName=Progress, Value=@n2)
   > n4 = GetDynamicFontMaterial!(execute, self:Text Object Reference=Get CurrentValue()) -> (then, ReturnValue:Material Instance Dynamic Object Reference)
   > SetScalarParameterValue(self=@n4, ParameterName=Progress, Value=@n2)
   SetDMGParticleFXTreatment!(execute, self, LastValue:double=GetLastValuePercentage(), CurrentValue:double=@n2) -> (then) {is=title K2Node_CallFunction}
  ]}
 UpdateCurrentValue {nodes [
   UpdateCurrentValue() {is=title K2Node_FunctionEntry}
   > StopAllAnimations(self:User Widget Object Reference)
   n1 = Break Widget Transform!(WidgetTransform:"Widget Transform Structure (by ref)"=Get RenderTransform!(self:Widget Object Reference=Get Overlay_Bar() -> (Overlay_Bar:Overlay Object Reference)) -> (RenderTransform:Widget Transform Structure)) -> (Translation:Vector 2D Structure, Scale:Vector 2D Structure, Shear:Vector 2D Structure, Angle:float=~) {is=title K2Node_BreakStruct}
   > SetRenderTransform!(
    execute
    self:Widget Object Reference=Get Overlay_Bar()
    InTransform:Widget Transform Structure=Make Widget Transform!(Translation:Vector 2D Structure=(X=0,Y=0), Scale:Vector 2D Structure=@n1.Scale, Shear:Vector 2D Structure=@n1.Shear, Angle:float=@n1.Angle) -> (WidgetTransform:Widget Transform Structure) {is=title K2Node_MakeStruct}
   ) -> (then)
   > n2 = Sequence!(execute) -> (then_0, then_1, then_2:Exec)
   Update Fill Bar(execute=@n2.then_1) {is=title K2Node_CallFunction}
   > "SetText (Text)"(
    self=Get CurrentValue()
    InText="To Text (Float)"(
     Value=Get ValueCurrent() -> (ValueCurrent:float)
     MinimumIntegralDigits=1
     MaximumIntegralDigits=324
     MinimumFractionalDigits=0
     MaximumFractionalDigits=0
    )
   )
   > Branch(Condition="Equal (Enum)"(
     A:EFortHitPointModificationReason Enum=Get ValueChangedReason() -> (ValueChangedReason:EFortHitPointModificationReason Enum=Invalid)
     B:EFortHitPointModificationReason Enum=InitalSet
    ))
   > ForceLayoutPrepass(self:Widget Object Reference=Get SizeBox_Value())
   n3 = Get BarType() -> (BarType=Health)
   Branch(
    execute=@n2.then_2
    Condition=OR Boolean(A="Equal (Enum)"(A:EHealthBarType Enum=@n3, B:EHealthBarType Enum=Shield), B="Equal (Enum)"(A=@n3, B=Health))
   )
   n4 = Get ValueCurrent()
   > n5 = Branch(Condition="float > float"(A=Get ValueLast() -> (ValueLast:float), B=@n4) {is=title K2Node_CallFunction})
   > PlayAnimation(InAnimation=Get Anim_ValueImpact())
   > PlayAnimation(InAnimation=Get Anim_BarShake(), NumLoopsToPlay=3, bRestoreState=true)
   n6 = SafeDivide(A=Get ValueCurrent(), B=Get ValueMax() -> (ValueMax:float))
   > Branch(Condition="float < float"!(A:double=@n6, B:double=0.5) -> (ReturnValue:bool) {is=title K2Node_CallFunction})
   > PlayAnimation(InAnimation=Get Anim_BarShake_1() -> (Anim_BarShake_1:Widget Animation Object Reference), NumLoopsToPlay=5, bRestoreState=true)
   > Branch(Condition="float < float"(A=@n6, B=0.2))
   > PlayAnimation(InAnimation=Get Anim_BarShake_2() -> (Anim_BarShake_2:Widget Animation Object Reference), NumLoopsToPlay=5, bRestoreState=true)
   Branch(execute=@n5.else, Condition="float >= float"(A=@n4, B=Get ValueMax()) {is=title K2Node_CallFunction})
   n7 = MakeLiteralInt!(Value:int=1) -> (ReturnValue:int=0)
   > SetNumLoopsToPlay!(execute, self:User Widget Object Reference, InAnimation:Widget Animation Object Reference=Get Anim_BarShake(), NumLoopsToPlay:int=@n7) -> (then)
   > SetNumLoopsToPlay(InAnimation=Get Anim_BarShake_1(), NumLoopsToPlay=@n7)
   > SetNumLoopsToPlay(InAnimation=Get Anim_BarShake_2(), NumLoopsToPlay=@n7)
   Branch(execute=@n2.then_0, Condition="Equal (Enum)"(A=Get BarType() -> (BarType=Health), B=Shield))
   n8 = NOT Boolean(A=Get bShouldDisplayShieldBar() -> (bShouldDisplayShieldBar:bool))
   n9 = AND Boolean(
    A="float <= float"!(A:double=Get ValueCurrent(), B:double=0) -> (ReturnValue:bool) {is=title K2Node_CallFunction}
    B=Get hideShieldOnEmpty() -> (hideShieldOnEmpty:bool)
   )
   > SetVisibility(InVisibility=Select!(Option 0:ESlateVisibility Enum=HitTestInvisible, Option 1:ESlateVisibility Enum=Collapsed, Index:bool=OR Boolean(A=@n9, B=@n8)) -> (ReturnValue:ESlateVisibility Enum))
  ]}
 UpdateDBNOState {nodes [
   n1 = UpdateDBNOState!() -> (then, bIsDBNO:bool=~) {is=title K2Node_FunctionEntry}
   > n2 = Branch(Condition="Equal (Boolean)"!(A:bool=@n1, B:bool=Get DBNOStateCache()) -> (ReturnValue:bool))
   n3 = Set DBNOStateCache(execute=@n2.else, DBNOStateCache:bool=@n1) -> (Output_Get:bool)
   > Switch on EHealthBarType!(execute, Selection:EHealthBarType Enum=Get BarType() -> (BarType=Health)) -> (Health:Exec, Shield:Exec, Overshield:Exec, Armor:Exec, Stamina:Exec, VehicleHealth:Exec, SignalInStorm:Exec) {is=title K2Node_SwitchEnum}
   > n4 = Branch(Condition=@n3)
   > n5 = For Each Loop(Array=Make Array!("[0]"=Get Image_Bar(), "[1]") -> (Array))
   SetBrushFromTexture!(
    execute=@n5.Completed
    self:Image Object Reference=Get Image_ValueIcon()
    Texture:Texture 2D Object Reference=/Game/Athena/HUD/Art/squadArt/squad_Eliminated
    bMatchSize:bool=false
   ) -> (then)
   > SetHealthStyle()
   n6 = GetDynamicMaterial(execute=@n5.LoopBody, self=@n5)
   > SetScalarParameterValue(self=@n6, ParameterName=DamageProgress, Value=0)
   SetHealthStyle(execute=@n4.else)
   > RefreshIcons() {is=title K2Node_CallFunction}
  ]}
 UpdateMaxValue {nodes [
   UpdateMaxValue() {is=title K2Node_FunctionEntry}
   > Update Fill Bar()
   > Update Delta Bar() {is=title K2Node_CallFunction}
  ]}
]
variables [
 AddSeparatorTag {type=GameplayTag, default {tagName=HUD.Athena.LocalInfoBarSeparator}}
 bAllowSeparator {type=boolean}
 BarHeight {type=number, default=32}
 BarWidth {type=number, default=344}
 bIcon_isVisible {type=boolean, default=true}
 bValue_isVisible {type=boolean, default=true}
 CachedInitialMaxValue {type=number}
 DBNOStateCache {type=boolean}
 DMGParticlesPosition {type=number}
 DMGParticlesVector2DOffset {type=Vector2D, default=<0 0>}
 FillGlowSize {type=number, default=32}
 GreenBarColor {type=LinearColor, default=<rgba 0 0.75 0 1>}
 HasReceivedDeltaUpdate {type=boolean}
 HasReceivedValueUpdate {type=boolean}
 hideShieldOnEmpty {type=boolean}
 IsFirstSetMaxHealth {type=boolean, default=true}
 IsViewTargetChanged {type=boolean}
 MaxValueWithSegments {type=number, default=150}
 PlayerController {type=FortPlayerController, default=null}
 Shield_Full {type=MaterialInterface, default=/Game/Athena/HUD/Materials/M_UI_HitPointBar_Shield}
 Shield_Segments {type=MaterialInterface, default=/Game/Athena/HUD/Materials/M_UI_HitPointBar_Shield_Segments}
 ShieldDelta_Full {type=MaterialInterface, default=/Game/Athena/HUD/Materials/M_UI_Delta_Shield}
 ShieldDelta_Segments {type=MaterialInterface, default=/Game/Athena/HUD/Materials/M_UI_HitPointBar_SegmentsBacking}
 TextHudElementTags {type=GameplayTagContainer, default {gameplayTags []}}
 TotalSegmentCount {type=number, default=6}
 ValueChangedReason {type=EFortHitPointModificationReason, default=Invalid}
 ViewTarget {type=Actor, default=null}
 WhiteBarColor {type=LinearColor, default=<rgba 0.3663 0.807 0.3372 1>}
]
widgets {
 info {
  inheritedWidgetCount=1
  namedSlotCount=1
  parentClass=/Script/FortniteUI.AthenaPlayerHitPointBarBase
  rootWidgetClass=&HorizontalBox
  widgetCount=20
 }
 widgets [
  HorizontalBox_66 {widgetClassPath=&HorizontalBox, widget=HorizontalBox_66}
  Overlay_0 {
   widgetClassPath=&Overlay
   parent=HorizontalBox_66
   widget=Overlay_0
   slot {class=HorizontalBoxSlot, properties {horizontalAlignment=HAlign_Center, verticalAlignment=VAlign_Center}}
  }
  Image_ValueIcon {
   widgetClassPath=&Image
   bIsVariable=true
   parent=Overlay_0
   widget=Image_ValueIcon
   properties {
    pixelSnapping=Disabled
    brush {
     bIsDynamicallyLoaded=false
     drawAs=Image
     imageType=NoImage
     mirroring=NoMirror
     resourceName=None
     resourceObject=&MI_UI_PlayerVitals_Icon_Health
     tiling=NoTile
     imageSize=<24 24>
     margin=<ltrb 0 0 0 0>
     outlineSettings {
      bUseBrushTransparency=false
      roundingType=HalfHeightRadius
      width=0
      color {colorUseRule=UseColor_Specified, specifiedColor=<rgba 0 0 0 0>}
      cornerRadii=<0 0 0 0>
     }
     tintColor {colorUseRule=UseColor_Specified, specifiedColor=<rgba 1 1 1 1>}
     uVRegion {bIsValid=false, max=<0 0>, min=<0 0>}
    }
    renderTransformPivot=<0.75 0.75>
   }
   slot {class=OverlaySlot, properties {horizontalAlignment=HAlign_Fill, verticalAlignment=VAlign_Fill}}
  }
  Bar+FX {
   widgetClassPath=&Overlay
   parent=HorizontalBox_66
   widget=Bar+FX
   slot {class=HorizontalBoxSlot, properties {padding=<ltrb 2 0 0 0>, size {sizeRule=Fill, value=1}}}
  }
  SizeBox_Bar {
   widgetClassPath=&SizeBox
   bIsVariable=true
   parent=Bar+FX
   widget=SizeBox_Bar
   properties {bOverride_HeightOverride=true, bOverride_WidthOverride=true, heightOverride=40, widthOverride=344}
   slot {class=OverlaySlot, properties {horizontalAlignment=HAlign_Fill, verticalAlignment=VAlign_Fill}}
  }
  Overlay_Bar {
   widgetClassPath=&Overlay
   bIsVariable=true
   parent=SizeBox_Bar
   widget=Overlay_Bar
   properties {visibility=HitTestInvisible}
   slot {class=SizeBoxSlot}
  }
  Image_Bar {
   widgetClassPath=&Image
   bIsVariable=true
   parent=Overlay_Bar
   widget=Image_Bar
   properties {
    visibility=SelfHitTestInvisible
    brush {
     bIsDynamicallyLoaded=false
     drawAs=Image
     imageType=FullColor
     mirroring=NoMirror
     resourceName=None
     resourceObject=&M_UI_HealthBar_Master
     tiling=NoTile
     imageSize=<32 32>
     margin=<ltrb 0 0 0 0>
     outlineSettings {
      bUseBrushTransparency=false
      roundingType=HalfHeightRadius
      width=0
      color {colorUseRule=UseColor_Specified, specifiedColor=<rgba 0 0 0 0>}
      cornerRadii=<0 0 0 0>
     }
     tintColor {colorUseRule=UseColor_Specified, specifiedColor=<rgba 1 1 1 1>}
     uVRegion {bIsValid=false, max=<0 0>, min=<0 0>}
    }
   }
   slot {class=OverlaySlot, properties {horizontalAlignment=HAlign_Fill, verticalAlignment=VAlign_Fill}}
  }
  Overlay_ExtensionWidget {
   widgetClassPath=&Overlay
   bIsVariable=true
   parent=Overlay_Bar
   widget=Overlay_ExtensionWidget
   slot {class=OverlaySlot, properties {horizontalAlignment=HAlign_Fill, verticalAlignment=VAlign_Fill}}
  }
  DMGParticles {
   widgetClassPath=&FortVisualAttachment
   bIsVariable=true
   parent=Bar+FX
   widget=DMGParticles
   properties {
    bOverride_HeightOverride=true
    bOverride_WidthOverride=true
    heightOverride=128
    visibility=Collapsed
    widthOverride=32
    contentAnchor=<0.5 0.5>
   }
   slot {class=OverlaySlot, properties {horizontalAlignment=HAlign_Right, verticalAlignment=VAlign_Center}}
  }
  ParticlesContainer {
   widgetClassPath=&Overlay
   parent=DMGParticles
   widget=ParticlesContainer
   properties {visibility=Hidden, renderTransform {angle=0, scale=<1 1>, shear=<-24 0>, translation=<0 0>}, renderTransformPivot=<0 0.5>}
   slot {class=SizeBoxSlot}
  }
  ParticlesVBox {
   widgetClassPath=&VerticalBox
   parent=ParticlesContainer
   widget=ParticlesVBox
   slot {class=OverlaySlot, properties {horizontalAlignment=HAlign_Fill, verticalAlignment=VAlign_Fill}}
  }
  DeltaParticles_Top {
   widgetClassPath=&Image
   bIsVariable=true
   parent=ParticlesVBox
   widget=DeltaParticles_Top
   properties {
    visibility=Hidden
    brush {
     bIsDynamicallyLoaded=false
     drawAs=Image
     imageType=FullColor
     mirroring=NoMirror
     resourceName=None
     resourceObject=&M_UI_PlayerVitals_DMGDelta_Particles
     tiling=NoTile
     imageSize=<32 32>
     margin=<ltrb 0 0 0 0>
     outlineSettings {
      bUseBrushTransparency=false
      roundingType=HalfHeightRadius
      width=0
      color {colorUseRule=UseColor_Specified, specifiedColor=<rgba 0 0 0 0>}
      cornerRadii=<0 0 0 0>
     }
     tintColor {colorUseRule=UseColor_Specified, specifiedColor=<rgba 1 1 1 1>}
     uVRegion {bIsValid=false, max=<0 0>, min=<0 0>}
    }
   }
   slot {class=VerticalBoxSlot, properties {size {sizeRule=Fill, value=0.5}}}
  }
  DeltaParticles_Bottom {
   widgetClassPath=&Image
   bIsVariable=true
   parent=ParticlesVBox
   widget=DeltaParticles_Bottom
   properties {
    visibility=Hidden
    brush {
     bIsDynamicallyLoaded=false
     drawAs=Image
     imageType=FullColor
     mirroring=NoMirror
     resourceName=None
     resourceObject=&M_UI_PlayerVitals_DMGDelta_Particles_Bottom
     tiling=NoTile
     imageSize=<32 32>
     margin=<ltrb 0 0 0 0>
     outlineSettings {
      bUseBrushTransparency=false
      roundingType=HalfHeightRadius
      width=0
      color {colorUseRule=UseColor_Specified, specifiedColor=<rgba 0 0 0 0>}
      cornerRadii=<0 0 0 0>
     }
     tintColor {colorUseRule=UseColor_Specified, specifiedColor=<rgba 1 1 1 1>}
     uVRegion {bIsValid=false, max=<0 0>, min=<0 0>}
    }
    renderTransform {angle=0, scale=<-1 -1>, shear=<0 0>, translation=<0 0>}
   }
   slot {class=VerticalBoxSlot, properties {size {sizeRule=Fill, value=0.5}}}
  }
  FortVisualAttachment_48 {
   widgetClassPath=&FortVisualAttachment
   parent=ParticlesContainer
   widget=FortVisualAttachment_48
   properties {
    bOverride_HeightOverride=true
    bOverride_WidthOverride=true
    heightOverride=128
    widthOverride=64
    contentAnchor=<0.5 1>
    renderTransform {angle=64, scale=<1 1>, shear=<0 0>, translation=<0 0>}
    renderTransformPivot=<0 0.5>
   }
   slot {class=OverlaySlot, properties {horizontalAlignment=HAlign_Center, verticalAlignment=VAlign_Center}}
  }
  DeltaParticles_Horizontal {
   widgetClassPath=&Image
   bIsVariable=true
   parent=FortVisualAttachment_48
   widget=DeltaParticles_Horizontal
   properties {
    visibility=Hidden
    brush {
     bIsDynamicallyLoaded=false
     drawAs=Image
     imageType=FullColor
     mirroring=NoMirror
     resourceName=None
     resourceObject=&MI_UI_PlayerVitals_DMGDelta_Particles_Horizontal
     tiling=NoTile
     imageSize=<32 32>
     margin=<ltrb 0 0 0 0>
     outlineSettings {
      bUseBrushTransparency=false
      roundingType=HalfHeightRadius
      width=0
      color {colorUseRule=UseColor_Specified, specifiedColor=<rgba 0 0 0 0>}
      cornerRadii=<0 0 0 0>
     }
     tintColor {colorUseRule=UseColor_Specified, specifiedColor=<rgba 1 1 1 1>}
     uVRegion {bIsValid=false, max=<0 0>, min=<0 0>}
    }
   }
   slot {class=SizeBoxSlot}
  }
  SizeBox_Value {
   widgetClassPath=&SizeBox
   bIsVariable=true
   parent=HorizontalBox_66
   widget=SizeBox_Value
   properties {bOverride_WidthOverride=true, widthOverride=200, renderTransform {angle=0, scale=<1 1>, shear=<0 0>, translation=<4 0>}}
   slot {class=HorizontalBoxSlot, properties {horizontalAlignment=HAlign_Left, verticalAlignment=VAlign_Center, padding=<ltrb 0 -6 -200 -6>}}
  }
  HBox_TextArea {
   widgetClassPath=&HorizontalBox
   bIsVariable=true
   parent=SizeBox_Value
   widget=HBox_TextArea
   slot {class=SizeBoxSlot, properties {verticalAlignment=VAlign_Center}}
  }
  FortHUDElementWrapper_Text {
   widgetClassPath=/Game/Athena/HUD/FortHUDElementWrapper
   bIsVariable=true
   parent=HBox_TextArea
   widget=FortHUDElementWrapper_Text
   properties {elementTag {tagName=HUD.Athena.Player.HitPointBarText}}
   slot {class=HorizontalBoxSlot}
  }
  CurrentValue {
   widgetClassPath=&CommonTextBlock
   bIsVariable=true
   namedSlotHost=FortHUDElementWrapper_Text
   widget=CurrentValue
   properties {
    justification=Right
    pixelSnapping=Disabled
    text=82
    font {
     bForceMonospaced=false
     bMaterialIsStencil=false
     fontMaterial=&M_Font_PlayerVitals_Value
     fontObject=&Burbank
     letterSpacing=30
     monospacedWidth=1
     size=16.5
     skewAmount=0
     typefaceFontName=BurbankRegular-Black
     outlineSettings {
      bApplyOutlineToDropShadows=true
      bMiteredCorners=false
      bSeparateFillAlpha=true
      outlineMaterial=/BattleRoyaleGameUI/Shared/MI_UI_Font_SoftShadow_SmallText
      outlineSize=4
      outlineColor=<rgba 1 1 1 1>
     }
    }
    renderTransformPivot=<0 0.75>
    shadowOffset=<0 0>
   }
  }
  TutorialHighlight {widgetClassPath=&TutorialHighlight}
 ]
}
