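# HUD widget showing a health bar for the viewed pawn's vehicle in Battle Royale, gated by a console variable.
# View-target changes rebind pawn vehicle-state and Creative quickbar delegates through a delta gate on the bound pawn.
# The bar hides without a vehicle, with a DisableHealth-tagged vehicle, or when the quickbar is equipped.
# With FadeOnFull, pawn health events play FadeVehicle when vehicle health equals 1, otherwise reverse it.
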WidgetBlueprint /BattleRoyaleGameUI/HUD/PlayerHitPointInfo/WBP_BRGameUI_VehicleHealth
animations [FadeVehicle {
  binding_count=1
  length_seconds=4
  object=&self:FadeVehicle
  sequence {bindings [WBP_BRGameUI_VehicleHealth_C_7 {
     id=#1
     possessed_class=WBP_BRGameUI_VehicleHealth_C
     tracks [{
       type=MovieSceneFloatTrack
       display_name=Render Opacity
       sections [{type=MovieSceneFloatSection, end_frame=80, start_frame=0, channels [None {type=MovieSceneScriptingFloatChannel, keys [[0,1],[40,1],[80,0]]}]}]
      }]
    }]}
 }]
graphs [
 EventGraph {nodes [
   EventOnPawnHealthChanged() {is=title K2Node_CustomEvent}
   > n1 = GetHealth!(
    execute:Exec
    self:Object Reference=GetVehicleActor!(self:Fort Player Pawn Object Reference=Get Bound Pawn() -> (Bound Pawn:Fort Player Pawn Object Reference)) -> (ReturnValue:Actor Object Reference)
   ) -> (then:Exec, ReturnValue:float) {is=title K2Node_Message}
   > Branch(Condition=IsVisible(self=Get HitPointBarVehicleHealth() -> (HitPointBarVehicleHealth:WBP BRGame UI Player Vital Bar Object Reference)))
   > n2 = Branch(Condition="Equal (Float)"!(A:double=@n1, B:double=1) -> (ReturnValue:bool) {is=title K2Node_CallFunction})
   n3 = Get FadeVehicle() -> (FadeVehicle:Widget Animation Object Reference)
   > PlayAnimation(InAnimation=@n3)
   PlayAnimationReverse(execute=@n2.else, InAnimation=@n3)
   n4 = VehicleStateChanged!() -> (OutputDelegate, then, FortPlayerPawn:Fort Player Pawn Object Reference, NewVehicle:Actor Object Reference, OldVehicle:Actor Object Reference) {is=title K2Node_CustomEvent}
   > n5 = Is Valid(InputObject=@n4.NewVehicle)
   ToggleVisibility!(execute=@n5.Is Not Valid, self, IsEnabled?:bool=false) -> (then)
   n6 = HasMatchingGameplayTag!(execute:Exec=@n5.Is Valid, self:Object Reference=@n4.NewVehicle, TagToCheck:Gameplay Tag Structure=(TagName="HUD.Vehicle.DisableHealth")) -> (then:Exec, ReturnValue:bool) {is=title K2Node_Message}
   > ToggleVisibility(IsEnabled?=NOT Boolean(A=@n6))
   Event Construct()
   > ToggleVisibility(IsEnabled?=false)
   > n7 = GetConsoleVariableBoolValue!(execute, VariableName:String=BattleRoyaleGameUI.VehicleHealthBar.Enabled) -> (then, ReturnValue:bool)
   > Branch(Condition=@n7)
   > n8 = Cast To FortPlayerController(Object=GetOwningPlayer())
   > Bind Event to On View Target Changed(self:Fort Player Controller Object Reference=@n8, Delegate=Create Event()) {is=title K2Node_AddDelegate}
   > OnViewTargetChanged!(
    execute
    self
    PlayerController:Fort Player Controller Object Reference
    OldViewTarget:Actor Object Reference
    NewViewTarget:Actor Object Reference=GetViewTarget!(self:Controller Object Reference=@n8) -> (ReturnValue:Actor Object Reference)
   ) -> (then)
   Event Destruct()
   > OnViewTargetChanged()
   > n9 = Cast To FortPlayerController(Object=GetOwningPlayer())
   > Unbind Event from On View Target Changed(self:Fort Player Controller Object Reference=@n9, Delegate=Create Event()) {is=title K2Node_RemoveDelegate}
   Event PreConstruct()
   Event Tick()
  ]}
 IsQuickbarEquipped {nodes [
   n1 = IsQuickbarEquipped!() -> (then, bIsQuickbarEquipped:bool=~) {is=title K2Node_FunctionEntry}
   > ToggleVisibility(IsEnabled?=NOT Boolean(A=@n1))
  ]}
 OnViewTargetChanged {nodes [
   n1 = OnViewTargetChanged!() -> (then, PlayerController:Fort Player Controller Object Reference, OldViewTarget:Actor Object Reference, NewViewTarget:Actor Object Reference) {is=title K2Node_FunctionEntry}
   > n2 = Delta Gate Wildcard!(
    execute:Exec
    Variable:"Fort Player Pawn Object Reference (by ref)"=Get Bound Pawn()
    New Value:Fort Player Pawn Object Reference=Cast To FortPlayerPawn!(Object=@n1.NewViewTarget) -> (AsFort Player Pawn, bSuccess:bool=~)
   ) -> (Success:Exec, Updated Variable:"Fort Player Pawn Object Reference (by ref)", Previous Variable:Fort Player Pawn Object Reference) {is=title K2Node_MacroInstance}
   > n3 = Sequence()
   > Is Valid(InputObject=@n2.Previous Variable)
   n4 = Create Event()
   > Unbind Event from On Vehicle State Changed Event(self:Fort Player Pawn Object Reference=@n2.Previous Variable, Delegate=@n4) {is=title K2Node_RemoveDelegate}
   n5 = Create Event()
   n6 = Get CreativeHUDLContext!() -> (ReturnValue:Creative HUDLContext Object Reference) {is=title K2Node_GetBlueprintContextEx}
   > Unbind Event from On Creative Quickbar Equipped Changed(self:Creative HUDLContext Object Reference=@n6, Delegate=@n5) {is=title K2Node_RemoveDelegate}
   n7 = Get FadeOnFull() -> (FadeOnFull:bool)
   > Branch(Condition=@n7)
   n8 = Create Event()
   n9 = Get Bound Pawn()
   > Unbind Event from On Pawn Health Changed(self:Fort Pawn Object Reference=@n9, Delegate=@n8) {is=title K2Node_RemoveDelegate}
   n10 = Get!(execute:Exec=@n3.then_1) -> (then:Exec, else:Exec, Bound Pawn:Fort Player Pawn Object Reference)
   > Bind Event to On Vehicle State Changed Event(self:Fort Player Pawn Object Reference=@n10, Delegate=@n4) {is=title K2Node_AddDelegate}
   > VehicleStateChanged!(
    execute
    self
    FortPlayerPawn:Fort Player Pawn Object Reference
    NewVehicle:Actor Object Reference=Get Vehicle!(self:Fort Player Pawn Object Reference=@n10) -> (ReturnValue:Fort Athena Vehicle Object Reference) {is=title K2Node_CallFunction}
    OldVehicle:Actor Object Reference
   ) -> (then) {is=title K2Node_CallFunction}
   > Bind Event to On Creative Quickbar Equipped Changed(self:Creative HUDLContext Object Reference=@n6, Delegate=@n5) {is=title K2Node_AddDelegate}
   > Branch(Condition=@n7)
   > Bind Event to On Pawn Health Changed(self:Fort Pawn Object Reference=@n9, Delegate=@n8) {is=title K2Node_AddDelegate}
   ToggleVisibility(execute=@n10.else, IsEnabled?=false)
  ]}
 ToggleVisibility {nodes [
   n1 = ToggleVisibility!() -> (then, IsEnabled?:bool=~) {is=title K2Node_FunctionEntry}
   > n2 = Branch(Condition=@n1)
   > n3 = SetVisibility(InVisibility=Visible)
   Return Node(execute=[@n3.then, @n4.then])
   n4 = SetVisibility(execute=@n2.else, InVisibility=Hidden)
  ]}
]
variables [
 Bound Pawn {type=FortPlayerPawn}
 FadeOnFull {type=boolean}
]
widgets {
 info {
  inheritedWidgetCount=1
  parentClass=&FortHUDElementWidget
  rootWidgetClass=&/WBP_BRGameUI_PlayerVitalBar.WBP_BRGameUI_PlayerVitalBar_C
  widgetCount=2
 }
 widgets [
  HitPointBarVehicleHealth {
   widgetClassPath=&/WBP_BRGameUI_PlayerVitalBar
   bIsVariable=true
   widget=HitPointBarVehicleHealth
   properties {
    barHeight=16
    barType=VehicleHealth
    barWidth=280
    fillGlowSize=24
    flowDirectionPreference=LeftToRight
    visibility=Collapsed
    elementTag {tagName=HUD.BottomLeft}
   }
  }
  TutorialHighlight {widgetClassPath=&TutorialHighlight}
 ]
}
